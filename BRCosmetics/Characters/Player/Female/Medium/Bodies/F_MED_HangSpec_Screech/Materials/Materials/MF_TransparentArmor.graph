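# Material function that gives the Screech character cosmetic a faux-glass transparent armor look, setting base color, surface, emissive, opacity and normal attributes.
# A custom-node spiral blur samples scene color with noise jitter, view-based refraction offset and depth-faded radius.
# It also builds fake sun highlights from refracted light-direction gradients, Fresnel thin-film iridescence and micro-normal detail.
# Emissive scales by time of day, and lower feature levels use flat color and opacity fallbacks.
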MaterialFunction /BRCosmetics/Characters/Player/Female/Medium/Bodies/F_MED_HangSpec_Screech/Materials/Materials/MF_TransparentArmor
graph {nodes [
  MaterialExpressionViewSize_2 = ViewSize()
  n1 = Convert(None=@MaterialExpressionViewSize_2) {
   nodeName=Break Float2
   convertInputs [{
     type=Vector2
     expressionInput {
      expression=&self:MaterialExpressionViewSize_2
      inputName=None
      mask=0
      maskA=0
      maskB=0
      maskG=0
      maskR=0
      outputIndex=0
     }
    }]
   convertOutputs [{type=Scalar}, {type=Scalar}, {type=Vector2}]
  }
  ComponentMask(None=MaterialFunctionCall(Texture Dimensions=@n1) {materialFunction=&ScreenAlignedUVs}."X 100%, Y 100%") {g=true, r=true}
  n2 = SkyAtmosphereLightDirection()
  n3 = Transform(None=NamedRerouteUsage())
  n4 = MaterialFunctionCall!(
   Ray Direction=@n2
   Surface Normal=@n3
   Refractive Index Origin
   Refractive Index Target=ScalarParameter() {defaultValue=1, group=TransparentArmor, parameterName=Refraction IOR}
  ) {materialFunction=/Engine/Functions/Engine_MaterialFunctions01/Vectors/Refract}
  ComponentMask(None=@n4.Vector) {b=true, r=true}
  n5 = SetMaterialAttributes!(
   MaterialAttributes
   BaseColor=NamedRerouteUsage()
   Metallic=ScalarParameter() {group=TransparentArmor, parameterName=Metallic}
   Specular=ScalarParameter() {defaultValue=1, group=TransparentArmor, parameterName=Specular}
   Roughness=ScalarParameter() {defaultValue=0.2, group=TransparentArmor, parameterName=Roughness}
   EmissiveColor=NamedRerouteUsage()
   Opacity=NamedRerouteUsage()
   Normal=NamedRerouteUsage()
  ) {attributeSetTypes [
    #1
    #2
    #3
    #4
    #5
    #6
    #7
   ]}
  FunctionOutput(None=@n5)
  n6 = TextureSampleParameter2D(UVs=Multiply(A=TextureCoordinate(), B=ScalarParameter() {defaultValue=200, group=TransparentArmor, parameterName=MicroNormalTiling}) {constB=200}) {
   group=TransparentArmor
   parameterName=MicroNormal
   samplerType=SAMPLERTYPE_Normal
   texture=/BRCosmetics/Characters/Player/Female/Medium/Bodies/F_MED_SoupGoal/Textures/T_SoupGoal_InteriorNoise_N
  }
  NamedRerouteDeclaration(None=@n6.RGB) {name=MicroNormal, nodeColor=<rgba 1 0.75 0 1>}
  n7 = TextureSampleParameter2D() {
   group=TransparentArmor
   parameterName=Normal
   samplerSource=SSM_Wrap_WorldGroupSettings
   samplerType=SAMPLERTYPE_Normal
   texture=&FlatNormalBlue
  }
  NamedRerouteDeclaration(None=@n7.RGB) {name=Normal_In, nodeColor=<rgba 0.4297 0 1 1>}
  NamedRerouteDeclaration(None=MaterialFunctionCall(X=Add(A=@n6.R, B=@n7.R), Y=Add(A=@n6.G, B=@n7.G), Z=Multiply(A=@n6.B, B=@n7.B)) {materialFunction=&MakeFloat3}.Result) {name=Normal_Out, nodeColor=<rgba 0 1 0.1094 1>}
  n8 = TextureSampleParameter2D(UVs=MaterialFunctionCall!(Vector=@n4.Vector) {materialFunction=/Engine/Functions/Engine_MaterialFunctions01/Coordinates/LongLatToUV}.Result) {
   group=TransparentArmor
   parameterName=RefractGradient
   samplerSource=SSM_Wrap_WorldGroupSettings
   texture=/Game/Characters/Player/Female/Medium/Bodies/F_MED_Critter_Streak/Textures/CritterStreak_Gradient
  }
  n9 = DotProduct(A=@n2, B=@n3)
  n10 = TextureSampleParameter2D(UVs=@n9) {
   group=TransparentArmor
   parameterName=RefractGradient
   samplerSource=SSM_Wrap_WorldGroupSettings
   texture=&CritterStreak_Gradient
  }
  n11 = ScalarParameter() {defaultValue=1, group=TransparentArmor, parameterName=SdN_Pow}
  n12 = Multiply(
   A=@n10.RGB
   B=Multiply(A=Power(Base=@n9, Exp=@n11), B=ScalarParameter() {defaultValue=10, group=TransparentArmor, parameterName=SdN_Mult}) {constB=0.5}
  )
  NamedRerouteDeclaration(None=LinearInterpolate(
    A=ScalarParameter() {group=TransparentArmor, parameterName=FakeHighLight_Low}
    B=ScalarParameter() {defaultValue=1, group=TransparentArmor, parameterName=FakeHighLight_High}
    Alpha=Max(A=@n8.RGB, B=@n12)
   )) {name=FakeLighting, nodeColor=<rgba 1 0 0 1>}
  n13 = Transform(None=NamedRerouteUsage())
  n14 = Multiply(
   A=NamedRerouteUsage()
   B=LinearInterpolate(
    A=VectorParameter() {group=TransparentArmor, parameterName=SurfaceEmissiveColor, defaultValue=<rgba 0.7588 0.974 0.9553 1>}.RGB
    Alpha=Fresnel(Normal=@n13) {exponent=0.2}
   ) {constB=0}
  )
  n15 = NamedRerouteUsage()
  n16 = ScalarParameter() {defaultValue=1, group=TransparentArmor, parameterName=Emissive_Final_Day}
  n17 = Multiply(
   A=Add(A=Add(A=@n14, B=@n15), B=NamedRerouteUsage())
   B=MaterialFunctionCall(
    Value_Morning=@n16
    Value_Day=@n16
    Value_Evening=@n16
    Value_Night=ScalarParameter() {defaultValue=0.5, group=TransparentArmor, parameterName=Emissive_Final_Night}
   ) {materialFunction=&Scalar_Time_of_Day}.Result
  )
  n18 = VectorParameter() {group=TransparentArmor, parameterName=SurfaceEmissiveColor_LowEnd, defaultValue=<rgba 0.4271 0.4566 0.6406 1>}
  NamedRerouteDeclaration(None=FeatureLevelSwitch(Default=@n17, ES2=@n18.RGB, ES3_1=@n18.RGB, SM5=@n17, SM6=@n17)) {name=Emissive, nodeColor=<rgba 0.2656 0 1 1>}
  n19 = NamedRerouteUsage()
  n20 = LocalPosition()
  n21 = Constant2Vector() {g=2, r=1}
  n22 = ScalarParameter() {defaultValue=1, group=TransparentArmor, parameterName=CylindricalUVTiling}
  n23 = TextureSample(UVs=Multiply(
    A=MaterialFunctionCall!(Circumference Tiling, Position=@n20.XYZ, Z Height, HQ Angle Calculation?) {materialFunction=/Game/Characters/Player/Common/MaterialFunctions/MF_CylindricalUVs_New}.UVs
    B=Multiply(A=@n21, B=@n22)
   )) {
   samplerType=SAMPLERTYPE_Normal
   texture=/BRCosmetics/Characters/Player/Female/Medium/Bodies/F_MED_BucketKick/Textures/T_BucketKick_SwirlNoise_N
  }
  n24 = MaterialFunctionCall(Normal=@n23.RGB, Flatness=ScalarParameter() {defaultValue=0.75, group=TransparentArmor, parameterName=CylindricalTiledNormalFlatten}) {materialFunction=&FlattenNormal}
  n25 = Fresnel(
   ExponentIn=ScalarParameter() {defaultValue=3, group=TransparentArmor, parameterName=ThinFilmExp}
   Normal=Transform(None=MaterialFunctionCall(BaseNormal=@n19, AdditionalNormal=@n24.Result) {materialFunction=&BlendAngleCorrectedNormals}.Result)
  ) {baseReflectFraction=0.1, exponent=3}
  n26 = TextureSampleParameter2D(UVs=@n25) {
   group=TransparentArmor
   parameterName=Thin Film Texture
   sortPriority=1
   texture=/Game/Global/Textures/Default/Thinfilm/T_ThinFilm_Spectrum_CoolColor
  }
  n27 = Multiply(A=@n25)
  NamedRerouteDeclaration(None=Desaturation(
    None=LinearInterpolate(B=@n26.RGB, Alpha=@n27)
    Fraction=ScalarParameter() {defaultValue=0.5, group=TransparentArmor, parameterName=ThinFilmDesaturation}
   )) {name=ThinFilm, nodeColor=<rgba 0.7812 0 1 1>}
  MaterialExpressionLinearInterpolate_2 = LinearInterpolate(
   A=ScalarParameter() {defaultValue=0.01, group=TransparentArmor, parameterName=SpiralBlur_Distance_Near}
   B=ScalarParameter() {defaultValue=0.003, group=TransparentArmor, parameterName=SpiralBlur_Distance_Far}
   Alpha=MaterialFunctionCall(Fade Length=ScalarParameter() {defaultValue=256, group=TransparentArmor, parameterName=SpiralBlur_FadingLength}) {materialFunction=&CameraDepthFade}.Result
  )
  MaterialExpressionScalarParameter_1 = ScalarParameter() {defaultValue=4, group=TransparentArmor, parameterName=SpiralBlur_Steps}
  MaterialExpressionScalarParameter_21 = ScalarParameter() {defaultValue=1, parameterName=SpiralBlur_RadialSteps}
  MaterialExpressionConstant_4 = Constant() {r=1}
  MaterialExpressionTextureObject_5 = TextureObject() {samplerType=SAMPLERTYPE_LinearGrayscale, texture=/Engine/EngineMaterials/Good64x64TilingNoiseHighFreq}
  MaterialExpressionScalarParameter_22 = ScalarParameter() {defaultValue=0.618, parameterName=SpiralBlur_RadialOffset}
  n28 = MaterialFunctionCall() {materialFunction=&ObjectPivotPoint}
  n29 = CameraPositionWS()
  MaterialExpressionMultiply_23 = Multiply(
   A=ScalarParameter() {defaultValue=0.02, parameterName=SpiralBlur_RefractionAmount}
   B=Normalize(VectorInput=Subtract(A=@n28.Object Pivot Location, B=@n29))
  )
  n30 = Custom!(
   Distance=@MaterialExpressionLinearInterpolate_2
   DistanceSteps=@MaterialExpressionScalarParameter_1
   RadialSteps=@MaterialExpressionScalarParameter_21
   TempAARotation=@MaterialExpressionConstant_4
   DistanceMask=@MaterialExpressionConstant_4
   Tex=@MaterialExpressionTextureObject_5
   TempAADistance=@MaterialExpressionConstant_4
   RadialOffset=@MaterialExpressionScalarParameter_22
   RefractionAmount=@MaterialExpressionMultiply_23
  ) {
   code="float3 CurColor = 0;\r\n\r\nfloat2 RawUV = MaterialFloat2(ScreenAlignedPosition(Parameters.ScreenPosition).xy);\r\n//float2 RefractedUV = BaseUV + (BaseUV.xy - 0.5) * RefractionAmount;\r\nfloat2 BaseUV = RawUV + (RawUV.xy - 0.5) * RefractionAmount;\r\nfloat2 NewUV = BaseUV;\r\nfloat StepSize = Distance / (int) DistanceSteps;\r\nfloat CurDistance = 0;\r\nfloat2 CurOffset = 0;\r\nfloat TwoPi = 6.283185;\r\nfloat Substep = 0;\r\nfloat2 ScenePixels=View.BufferSizeAndInvSize.xy*BaseUV;\r\nScenePixels+=View.TemporalAAParams.r;\r\nfloat2 RandomSamp = ((uint)(ScenePixels.x) + 2 * (uint)(ScenePixels.y)) % 5;\r\nRandomSamp+=Texture2DSample(Tex,TexSampler,ScenePixels);\r\nRandomSamp/=5;\r\nRandomSamp-=0.5;\r\nTempAARotation*=RandomSamp;\r\nTempAADistance*=StepSize*RandomSamp;\r\n\r\n\r\nint i=0;\r\nif (DistanceSteps<1)\r\n{\r\nreturn DecodeSceneColorForMaterialNode(NewUV);\r\n}\r\nelse\r\n{\r\nfloat2 BilinearOffset = View.BufferSizeAndInvSize.zw * 0.5f;\r\nfloat2 BilinearMinBufferUV = ViewportUVToBufferUV(float2(0.0f, 0.0f)) + BilinearOffset;\r\nfloat2 BilinearMaxBufferUV = ViewportUVToBufferUV(float2(1.0f, 1.0f)) - BilinearOffset;\r\n//CurDistance += 0.5*StepSize;\r\nwhile ( i < (int) DistanceSteps)\r\n{\r\n\r\n//CurDistance+=StepSize;\r\nfor (int j = 0; j < (int) RadialSteps; j++)\r\n{\r\nCurOffset.x = cos(TwoPi*((TempAARotation+Substep) / RadialSteps));\r\nCurOffset.y = sin(TwoPi*((TempAARotation+Substep) / RadialSteps));\r\nCurOffset *=DistanceMask;\r\nNewUV.x = BaseUV.x + (CurOffset.x * (CurDistance+(RandomSamp*TempAADistance)));\r\nNewUV.y = BaseUV.y + (CurOffset.y * (CurDistance+(RandomSamp*TempAADistance)));\r\nNewUV = clamp(NewUV, BilinearMinBufferUV, BilinearMaxBufferUV);\r\nCurColor += DecodeSceneColorForMaterialNode(NewUV);\r\n//CurDistance+=(StepSize+(TempAADistance))/RadialSteps;\r\nSubstep++;\r\n}\r\nCurDistance+=StepSize;\r\nSubstep+=RadialOffset;\r\ni++;\r\n}\r\nCurColor = CurColor / ((int)DistanceSteps*(int)RadialSteps);\r\nreturn CurColor;\r\n}"
   desc="float3 CurColor = 0;\r\nfloat2 BaseUV = MaterialFloat2(ScreenAlignedPosition(Parameters.ScreenPosition).xy);\r\nfloat2 NewUV = BaseUV;\r\nfloat StepSize = Distance / (int) DistanceSteps;\r\nfloat CurDistance = 0;\r\nfloat2 CurOffset = 0;\r\nfloat TwoPi = 6.283185;\r\nfloat Substep = 0;\r\nfloat2 ScenePixels=View.BufferSizeAndInvSize.xy*BaseUV;\r\nScenePixels+=View.TemporalAAParams.r;\r\nfloat2 RandomSamp = ((uint)(ScenePixels.x) + 2 * (uint)(ScenePixels.y)) % 5;\r\nRandomSamp+=Texture2DSample(Tex,TexSampler,ScenePixels);\r\nRandomSamp/=5;\r\nRandomSamp-=0.5;\r\nTempAARotation*=RandomSamp;\r\nTempAADistance*=StepSize*RandomSamp;\r\n\r\n\r\nint i=0;\r\nif (DistanceSteps<1)\r\n{\r\nreturn DecodeSceneColorForMaterialNode(NewUV);\r\n}\r\nelse\r\n{\r\nfloat2 BilinearOffset = View.BufferSizeAndInvSize.zw * 0.5f;\r\nfloat2 BilinearMinBufferUV = ViewportUVToBufferUV(float2(0.0f, 0.0f)) + BilinearOffset;\r\nfloat2 BilinearMaxBufferUV = ViewportUVToBufferUV(float2(1.0f, 1.0f)) - BilinearOffset;\r\n//CurDistance += 0.5*StepSize;\r\nwhile ( i < (int) DistanceSteps)\r\n{\r\n\r\n//CurDistance+=StepSize;\r\nfor (int j = 0; j < (int) RadialSteps; j++)\r\n{\r\nCurOffset.x = cos(TwoPi*((TempAARotation+Substep) / RadialSteps));\r\nCurOffset.y = sin(TwoPi*((TempAARotation+Substep) / RadialSteps));\r\nCurOffset *=DistanceMask;\r\nNewUV.x = BaseUV.x + (CurOffset.x * (CurDistance+(RandomSamp*TempAADistance)));\r\nNewUV.y = BaseUV.y + (CurOffset.y * (CurDistance+(RandomSamp*TempAADistance)));\r\nNewUV = clamp(NewUV, BilinearMinBufferUV, BilinearMaxBufferUV);\r\nCurColor += DecodeSceneColorForMaterialNode(NewUV);\r\n//CurDistance+=(StepSize+(TempAADistance))/RadialSteps;\r\nSubstep++;\r\n}\r\nCurDistance+=StepSize;\r\nSubstep+=RadialOffset;\r\ni++;\r\n}\r\nCurColor = CurColor / ((int)DistanceSteps*(int)RadialSteps);\r\nreturn CurColor;\r\n}"
   showCode=true
   inputs [
    {
     inputName=Distance
     input {expression=&self:MaterialExpressionLinearInterpolate_2}
    }
    {
     inputName=DistanceSteps
     input {expression=&self:MaterialExpressionScalarParameter_1}
    }
    {
     inputName=RadialSteps
     input {expression=&self:MaterialExpressionScalarParameter_21}
    }
    {
     inputName=TempAARotation
     input {expression=&self:MaterialExpressionConstant_4}
    }
    {
     inputName=DistanceMask
     input {expression=&self:MaterialExpressionConstant_4}
    }
    {
     inputName=Tex
     input {expression=&self:MaterialExpressionTextureObject_5}
    }
    {
     inputName=TempAADistance
     input {expression=&self:MaterialExpressionConstant_4}
    }
    {
     inputName=RadialOffset
     input {expression=&self:MaterialExpressionScalarParameter_22}
    }
    {
     inputName=RefractionAmount
     input {expression=&self:MaterialExpressionMultiply_23}
    }
   ]
  }
  NamedRerouteDeclaration(None=@n30) {name=BlurredColor, nodeColor=<rgba 0.0547 1 0 1>}
  n31 = Multiply(
   A=NamedRerouteUsage()
   B=VectorParameter() {group=TransparentArmor, parameterName=SurfaceBaseColor, defaultValue=<rgba 0.4271 0.4566 0.6406 1>}.RGB
  )
  n32 = VectorParameter() {group=TransparentArmor, parameterName=SurfaceBaseColor_LowEnd, defaultValue=<rgba 0.4271 0.4566 0.6406 1>}
  NamedRerouteDeclaration(None=FeatureLevelSwitch(Default=@n31, ES2=@n32.RGB, ES3_1=@n32.RGB, SM5=@n31, SM6=@n31)) {name=BaseColor, nodeColor=<rgba 1 0 0.375 1>}
  n33 = Multiply(A=Max(A=SceneColor!(UVs).RGB), B=ScalarParameter() {defaultValue=1, parameterName=Opacity})
  n34 = ScalarParameter() {defaultValue=0.5, group=TransparentArmor, parameterName=MobileOpacity}
  NamedRerouteDeclaration(None=FeatureLevelSwitch(Default=@n33, ES2=@n34, ES3_1=@n34, SM5=@n33, SM6=@n33)) {name=Opacity, nodeColor=<rgba 0 0.6953 1 1>}
 ]}
