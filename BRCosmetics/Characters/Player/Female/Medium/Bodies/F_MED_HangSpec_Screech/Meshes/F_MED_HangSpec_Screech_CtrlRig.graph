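# Control Rig for the F_MED_HangSpec_Screech player body that corrects outfit helper bones on top of the animated skeleton.
# It runs the shared DeformRig function, then weighted parent constraints for jacket, zips, skirt, belt and ribbons.
# DeformRig alpha attributes drive elbow, knee and belt offsets, while smoothed, curve-remapped swing of the right thigh and spine adjusts belt bones.
# Final fixed offsets follow, with emote and front-end flags adding tweaks.
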ControlRigBlueprint /BRCosmetics/Characters/Player/Female/Medium/Bodies/F_MED_HangSpec_Screech/Meshes/F_MED_HangSpec_Screech_CtrlRig
graphs [
 New Function_ContainedGraph {nodes [
   RigVMFunctionEntryNode()
   > &RigUnit_ModifyTransforms(
    ItemToModify=((Item=(Type=Bone,Name="lowerarm_twist_02_l"),Transform=(Rotation=(X=0,Y=0,Z=0,W=1),Translation=(X=3,Y=0,Z=0),Scale3D=(X=1,Y=1,Z=1))))
    WeightMinimum=0
    Weight=RigVMDispatchNode(BoneName=root, Default=0, Name=deformRig_l_elbow_alpha) {template="DISPATCH_RigDispatch_GetAnimAttribute(in Name,in BoneName,in Default,out Value,out Found)"}.Value
   )
   > &RigUnit_ModifyTransforms(
    ItemToModify="((Item=(Type=Bone,Name=\"base_knee_pad_l\"),Transform=(Rotation=(X=0,Y=0,Z=-0.3827,W=0.9239),Translation=(X=-0.5,Y=-2,Z=0.5),Scale3D=(X=1,Y=1,Z=1.15))),(Item=(Type=Bone,Name=\"base_ribbon_leg_A_l\"),Transform=(Rotation=(X=0,Y=-0.2164,Z=0,W=0.9763),Translation= ...(truncated)"
    WeightMinimum=0
    Weight=RigVMDispatchNode(BoneName=root, Default=0, Name=deformRig_l_knee_alpha) {template="DISPATCH_RigDispatch_GetAnimAttribute(in Name,in BoneName,in Default,out Value,out Found)"}.Value
   )
   > &RigUnit_ModifyTransforms(
    ItemToModify="((Item=(Type=Bone,Name=\"base_knee_pad_r\"),Transform=(Rotation=(X=0,Y=0,Z=-0.3827,W=0.9239),Translation=(X=0.5,Y=2,Z=-0.5),Scale3D=(X=1,Y=1,Z=1.15))),(Item=(Type=Bone,Name=\"base_ribbon_leg_A_r\"),Transform=(Rotation=(X=0,Y=-0.2164,Z=0,W=0.9763),Translation=( ...(truncated)"
    WeightMinimum=0
    Weight=RigVMDispatchNode(BoneName=root, Default=0, Name=deformRig_r_knee_alpha) {template="DISPATCH_RigDispatch_GetAnimAttribute(in Name,in BoneName,in Default,out Value,out Found)"}.Value
   )
   > RigVMFunctionReturnNode()
  ]}
 New Function_ContainedGraph {nodes [
   RigVMFunctionEntryNode()
   > RigVMFunction_Sequence = RigVMAggregateNode()
   &RigUnit_ParentConstraint(
    AdvancedSettings=(RotationOrderForFilter=XZY)
    Child=(Type=Bone,Name="belt_support_front_r")
    Parents="((Item=(Type=Bone,Name=\"thigh_r\"),Weight=0.0003),(Item=(Type=Bone,Name=\"spine_04\"),Weight=0.0006),(Item=(Type=Bone,Name=\"spine_02\"),Weight=0.0222),(Item=(Type=Bone,Name=\"thigh_twist_01_r\"),Weight=0.1284),(Item=(Type=Bone,Name=\"deform_groin_r\"),Weight=0.150 ...(truncated)"
    Weight=1
    ExecutePin=@RigVMFunction_Sequence.A
   )
   n1 = &RigUnit_GetRelativeTransformForItem(Child=(Type=Bone,Name="thigh_twist_01_r"), Parent=(Type=Control,Name="thigh_twist_01_r_ctrl"), bChildInitial=False, bParentInitial=False)
   RigVMFunction_MathQuaternionSwingTwist = &RigVMFunction_MathQuaternionSwingTwist(TwistAxis=(X=1,Y=0,Z=0), Input=@n1.RelativeTransform.Rotation)
   RigVMFunction_MathQuaternionToEuler = &RigVMFunction_MathQuaternionToEuler(RotationOrder=ZYX, Value=@RigVMFunction_MathQuaternionSwingTwist.Swing)
   RigVMVariableNode(Value=@RigVMFunction_MathQuaternionToEuler.Result, ExecuteContext=@RigVMFunction_Sequence.B) {variable=Leg_r_Swing_rot}
   n2 = &RigUnit_GetRelativeTransformForItem(Child=(Type=Bone,Name="pelvis"), Parent=(Type=Bone,Name="spine_05"), bChildInitial=False, bParentInitial=False)
   > RigVMVariableNode(Value=&RigVMFunction_MathQuaternionToEuler(
     RotationOrder=ZYX
     Value=&RigVMFunction_MathQuaternionSwingTwist(TwistAxis=(X=1,Y=0,Z=0), Input=@n2.RelativeTransform.Rotation).Swing
    ).Result) {variable=Spine_Rotation}
   RigVMFunctionReturnNode(ExecuteContext=@RigVMFunction_Sequence.D)
   n3 = RigVMVariableNode() {variable=Leg_r_Swing_rot}
   n4 = &RigVMFunction_AnimEvalRichCurve(
    Curve="(EditorCurveData=(Keys=((InterpMode=RCIM_Cubic,TangentMode=RCTM_User,TangentWeightMode=RCTWM_WeightedNone,Time=0,Value=0,ArriveTangent=0.2796,ArriveTangentWeight=0,LeaveTangent=0.2796,LeaveTangentWeight=0),(InterpMode=RCIM_Cubic,TangentMode=RCTM_User,Tange ...(truncated)"
    SourceMaximum=-140
    SourceMinimum=0
    TargetMaximum=-44
    TargetMinimum=0
    Value=@n3.Value.Y
   )
   n5 = &RigVMFunction_AnimEvalRichCurve(
    Curve="(EditorCurveData=(Keys=((InterpMode=RCIM_Cubic,TangentMode=RCTM_User,TangentWeightMode=RCTWM_WeightedNone,Time=0,Value=0,ArriveTangent=0.2796,ArriveTangentWeight=0,LeaveTangent=0.2796,LeaveTangentWeight=0),(InterpMode=RCIM_Cubic,TangentMode=RCTM_User,Tange ...(truncated)"
    SourceMaximum=-140
    SourceMinimum=0
    TargetMaximum=-44
    TargetMinimum=0
    Value=@n3.Value.Z
   )
   n6 = &RigVMFunction_AlphaInterp(
    InterpSpeedDecreasing=10
    InterpSpeedIncreasing=10
    bClampResult=false
    bMapRange=true
    Value=&RigVMFunction_MathDoubleAdd(A=@n4.Result, B=@n5.Result).Result
    Scale=RigVMDispatchNode(False=2, True=1, Condition=RigVMVariableNode() {variable=isPlayingEmotes}.Value) {template="DISPATCH_RigVMDispatch_If(in Condition,in True,in False,out Result)"}.Result
   )
   n7 = &RigVMFunction_MathQuaternionFromEuler(Euler=(X=0,Y=0,Z=0), Euler.Y=&RigVMFunction_MathDoubleMul(B=1, A=@n6.Result).Result)
   &RigUnit_ModifyTransforms(
    ItemToModify="((Item=(Type=Bone,Name=\"dyn_belt_ammo_1\"),Transform=(Rotation=(X=0,Y=0,Z=0,W=1),Translation=(X=0,Y=0,Z=0),Scale3D=(X=1,Y=1,Z=1))),(Item=(Type=Bone,Name=\"dyn_belt_ammo_2\"),Transform=(Rotation=(X=0,Y=0,Z=0,W=1),Translation=(X=0,Y=0,Z=0),Scale3D=(X=1,Y=1,Z=1) ...(truncated)"
    Weight=1
    WeightMinimum=0
    ItemToModify.1.Transform.Rotation=@n7.Result
    ItemToModify.2.Transform.Rotation=@n7.Result
    ItemToModify.0.Transform.Rotation=&RigVMFunction_MathQuaternionFromEuler(Euler=(X=0,Y=0,Z=0), Euler.Y=&RigVMFunction_MathDoubleMul(B=1, A=@n6.Result).Result).Result
    ExecutePin=@RigVMFunction_Sequence.C
   )
   n8 = RigVMDispatchNode(BoneName=root, Default=0, Name=deformRig_r_thigh_up_alpha) {template="DISPATCH_RigDispatch_GetAnimAttribute(in Name,in BoneName,in Default,out Value,out Found)"}
   Multiply = &RigVMFunction_MathDoubleMul(B=0.5, A=@n8.Value)
   > &RigUnit_ModifyTransforms(
    ItemToModify="((Item=(Type=Bone,Name=\"belt_center_r\"),Transform=(Rotation=(X=0,Y=0.1305,Z=0,W=0.9914),Translation=(X=2,Y=0,Z=0),Scale3D=(X=1,Y=1,Z=1.1))),(Item=(Type=Bone,Name=\"belt_support_front_r\"),Transform=(Rotation=(X=0,Y=0,Z=0,W=1),Translation=(X=0.25,Y=0,Z=0),Sca ...(truncated)"
    WeightMinimum=0
    Weight=@Multiply.Result
   )
   > &RigUnit_ModifyTransforms(
    ItemToModify="((Item=(Type=Bone,Name=\"belt_center_r\"),Transform=(Rotation=(X=-0.0872,Y=0,Z=0,W=0.9962),Translation=(X=0,Y=0,Z=0),Scale3D=(X=1.1,Y=1,Z=1))),(Item=(Type=Bone,Name=\"belt_support_front_r\"),Transform=(Rotation=(X=0,Y=0,Z=0,W=1),Translation=(X=0,Y=0,Z=0),Scale ...(truncated)"
    Weight=1
    WeightMinimum=0
    ItemToModify.2.Transform.Translation=RigVMDispatchNode(False=(X=0,Y=0,Z=0), True=(X=0,Y=0,Z=2), Condition=RigVMVariableNode() {variable=isPlayingEmotes}.Value) {template="DISPATCH_RigVMDispatch_If(in Condition,in True,in False,out Result)"}.Result
   )
   > &RigUnit_ModifyTransforms(
    ItemToModify=((Item=(Type=Bone,Name="base_belt_support_back_r"),Transform=(Rotation=(X=0,Y=0,Z=0,W=1),Translation=(X=0,Y=0,Z=4),Scale3D=(X=1,Y=1,Z=1))))
    WeightMinimum=0
    Weight=@n8.Value
   )
   n9 = RigVMVariableNode() {variable=Spine_Rotation}
   n10 = &RigVMFunction_AnimEvalRichCurve(
    Curve="(EditorCurveData=(Keys=((InterpMode=RCIM_Cubic,TangentMode=RCTM_User,TangentWeightMode=RCTWM_WeightedNone,Time=0,Value=0,ArriveTangent=0.2796,ArriveTangentWeight=0,LeaveTangent=0.2796,LeaveTangentWeight=0),(InterpMode=RCIM_Cubic,TangentMode=RCTM_User,Tange ...(truncated)"
    SourceMaximum=-140
    SourceMinimum=0
    TargetMaximum=-44
    TargetMinimum=0
    Value=@n9.Value.Y
   )
   n11 = &RigVMFunction_AnimEvalRichCurve(
    Curve="(EditorCurveData=(Keys=((InterpMode=RCIM_Cubic,TangentMode=RCTM_User,TangentWeightMode=RCTWM_WeightedNone,Time=0,Value=0,ArriveTangent=0.2796,ArriveTangentWeight=0,LeaveTangent=0.2796,LeaveTangentWeight=0),(InterpMode=RCIM_Cubic,TangentMode=RCTM_User,Tange ...(truncated)"
    SourceMaximum=-140
    SourceMinimum=0
    TargetMaximum=-44
    TargetMinimum=0
    Value=@n9.Value.Z
   )
   n12 = &RigVMFunction_AlphaInterp(
    InterpSpeedDecreasing=10
    InterpSpeedIncreasing=10
    Scale=1
    bClampResult=false
    bInterpResult=false
    bMapRange=true
    Value=&RigVMFunction_MathDoubleAdd(A=@n10.Result, B=@n11.Result).Result
   )
   > &RigUnit_ModifyTransforms(
    ItemToModify=((Item=(Type=Bone,Name="base_belt_support_front_l"),Transform=(Rotation=(X=0,Y=0,Z=0,W=1),Translation=(X=0,Y=0,Z=4),Scale3D=(X=1,Y=1,Z=1))))
    WeightMinimum=0
    Weight=&RigVMFunction_MathDoubleMul(B=-0.3, A=@n12.Result).Result
   )
   > &RigUnit_ModifyTransforms(
    ItemToModify=((Item=(Type=Bone,Name="base_ribbon"),Transform=(Rotation=(X=-0.0868,Y=0.0868,Z=0.0076,W=0.9924),Translation=(X=-0.25,Y=0,Z=-0.8),Scale3D=(X=1,Y=1,Z=1))))
    Weight=1
    WeightMinimum=0
   )
  ]}
 New Function_ContainedGraph {nodes [
   n1 = RigVMFunctionEntryNode()
   RigVMFunction_Sequence = &RigVMFunction_Sequence(ExecuteContext=@n1.ExecuteContext)
   &RigUnit_ParentConstraint(
    AdvancedSettings=(RotationOrderForFilter=XZY)
    Child=(Type=Bone,Name="base_zip_neck_1")
    Parents="((Item=(Type=Bone,Name=\"spine_03\"),Weight=0.0055),(Item=(Type=Bone,Name=\"spine_04\"),Weight=0.0978),(Item=(Type=Bone,Name=\"head\"),Weight=0.104),(Item=(Type=Bone,Name=\"neck_01\"),Weight=0.1589),(Item=(Type=Bone,Name=\"neck_02\"),Weight=0.1996),(Item=(Type=Bone, ...(truncated)"
    Weight=1
    ExecutePin=@RigVMFunction_Sequence.A
   )
   > &RigUnit_ParentConstraint(
    AdvancedSettings=(RotationOrderForFilter=XZY)
    Child=(Type=Bone,Name="base_zip_neck_2")
    Parents="((Item=(Type=Bone,Name=\"neck_02\"),Weight=0.011),(Item=(Type=Bone,Name=\"clavicle_r\"),Weight=0.014),(Item=(Type=Bone,Name=\"clavicle_l\"),Weight=0.0167),(Item=(Type=Bone,Name=\"spine_03\"),Weight=0.0236),(Item=(Type=Bone,Name=\"neck_01\"),Weight=0.0414),(Item=(Typ ...(truncated)"
    Weight=1
   )
   > &RigUnit_ParentConstraint(
    AdvancedSettings=(RotationOrderForFilter=XZY)
    Child=(Type=Bone,Name="base_jacket_zip_fr_l")
    Parents="((Item=(Type=Bone,Name=\"deform_upperarm_fr_l\"),Weight=0.0003),(Item=(Type=Bone,Name=\"upperarm_twist_01_l\"),Weight=0.0019),(Item=(Type=Bone,Name=\"clavicle_l\"),Weight=0.0568),(Item=(Type=Bone,Name=\"dyn_jacket_fr_l\"),Weight=0.0809),(Item=(Type=Bone,Name=\"spin ...(truncated)"
    Weight=1
   )
   > &RigUnit_ParentConstraint(
    AdvancedSettings=(RotationOrderForFilter=XZY)
    Child=(Type=Bone,Name="base_jacket_zip_fr_r")
    Parents="((Item=(Type=Bone,Name=\"deform_upperarm_fr_r\"),Weight=0.0003),(Item=(Type=Bone,Name=\"upperarm_twist_01_r\"),Weight=0.0021),(Item=(Type=Bone,Name=\"clavicle_r\"),Weight=0.0618),(Item=(Type=Bone,Name=\"spine_05\"),Weight=0.2813),(Item=(Type=Bone,Name=\"spine_04\"), ...(truncated)"
    Weight=1
   )
   > &RigUnit_ParentConstraint(
    AdvancedSettings=(RotationOrderForFilter=XZY)
    Child=(Type=Bone,Name="base_ribbon")
    Parents="((Item=(Type=Bone,Name=\"belt_support_front_r\"),Weight=0.0003),(Item=(Type=Bone,Name=\"spine_05\"),Weight=0.0004),(Item=(Type=Bone,Name=\"spine_04\"),Weight=0.0034),(Item=(Type=Bone,Name=\"spine_03\"),Weight=0.0233),(Item=(Type=Bone,Name=\"pelvis\"),Weight=0.0492), ...(truncated)"
    Weight=1
   )
   > &RigUnit_ParentConstraint(
    AdvancedSettings=(RotationOrderForFilter=XZY)
    Child=(Type=Bone,Name="base_ribbon_leg_B_l")
    Parents=((Item=(Type=Bone,Name="thigh_l"),Weight=0.0132),(Item=(Type=Bone,Name="calf_twist_02_l"),Weight=0.1916),(Item=(Type=Bone,Name="calf_l"),Weight=0.7953))
    Weight=1
   )
   > &RigUnit_ParentConstraint(
    AdvancedSettings=(RotationOrderForFilter=XZY)
    Child=(Type=Bone,Name="base_ribbon_leg_B_r")
    Parents=((Item=(Type=Bone,Name="thigh_r"),Weight=0.0132),(Item=(Type=Bone,Name="calf_twist_02_r"),Weight=0.1916),(Item=(Type=Bone,Name="calf_r"),Weight=0.7953))
    Weight=1
   )
   > &RigUnit_ParentConstraint(
    AdvancedSettings=(RotationOrderForFilter=XZY)
    Child=(Type=Bone,Name="base_ribbon_knee_A_r")
    Parents=((Item=(Type=Bone,Name="calf_r"),Weight=0.0582),(Item=(Type=Bone,Name="deform_knee_out_r"),Weight=0.1889),(Item=(Type=Bone,Name="thigh_r"),Weight=0.7529))
    Weight=1
   )
   > &RigUnit_ParentConstraint(
    AdvancedSettings=(RotationOrderForFilter=XZY)
    Child=(Type=Bone,Name="base_ribbon_knee_A_l")
    Parents=((Item=(Type=Bone,Name="calf_l"),Weight=0.0582),(Item=(Type=Bone,Name="deform_knee_out_l"),Weight=0.1889),(Item=(Type=Bone,Name="thigh_l"),Weight=0.7529))
    Weight=1
   )
   > &RigUnit_ParentConstraint(
    AdvancedSettings=(RotationOrderForFilter=XZY)
    Child=(Type=Bone,Name="base_front_skirt_l")
    Parents="((Item=(Type=Bone,Name=\"deform_groin_r\"),Weight=0.0172),(Item=(Type=Bone,Name=\"thigh_r\"),Weight=0.0227),(Item=(Type=Bone,Name=\"thigh_twist_01_r\"),Weight=0.0553),(Item=(Type=Bone,Name=\"deform_groin_l\"),Weight=0.1503),(Item=(Type=Bone,Name=\"thigh_l\"),Weight= ...(truncated)"
    Weight=1
   )
   > &RigUnit_ParentConstraint(
    AdvancedSettings=(RotationOrderForFilter=XZY)
    Child=(Type=Bone,Name="base_front_skirt_r")
    Parents="((Item=(Type=Bone,Name=\"deform_groin_l\"),Weight=0.0168),(Item=(Type=Bone,Name=\"thigh_l\"),Weight=0.0222),(Item=(Type=Bone,Name=\"thigh_twist_01_l\"),Weight=0.0544),(Item=(Type=Bone,Name=\"deform_groin_r\"),Weight=0.1509),(Item=(Type=Bone,Name=\"thigh_r\"),Weight= ...(truncated)"
    Weight=1
   )
   > &RigUnit_ParentConstraint(
    AdvancedSettings=(RotationOrderForFilter=XZY)
    Child=(Type=Bone,Name="base_side_skirt_l")
    Parents="((Item=(Type=Bone,Name=\"deform_glute_l\"),Weight=0.0135),(Item=(Type=Bone,Name=\"deform_groin_l\"),Weight=0.032),(Item=(Type=Bone,Name=\"pelvis\"),Weight=0.1368),(Item=(Type=Bone,Name=\"thigh_l\"),Weight=0.1561),(Item=(Type=Bone,Name=\"thigh_twist_01_l\"),Weight=0. ...(truncated)"
    Weight=1
   )
   > &RigUnit_ParentConstraint(
    AdvancedSettings=(RotationOrderForFilter=XZY)
    Child=(Type=Bone,Name="base_side_skirt_r")
    Parents="((Item=(Type=Bone,Name=\"deform_glute_r\"),Weight=0.013),(Item=(Type=Bone,Name=\"deform_groin_r\"),Weight=0.0346),(Item=(Type=Bone,Name=\"pelvis\"),Weight=0.1363),(Item=(Type=Bone,Name=\"thigh_r\"),Weight=0.1571),(Item=(Type=Bone,Name=\"thigh_twist_01_r\"),Weight=0. ...(truncated)"
    Weight=1
   )
   > &RigUnit_ParentConstraint(
    AdvancedSettings=(RotationOrderForFilter=XZY)
    Child=(Type=Bone,Name="base_bk_skirt_r")
    Parents=((Item=(Type=Bone,Name="thigh_twist_01_r"),Weight=0.1027),(Item=(Type=Bone,Name="deform_glute_r"),Weight=0.3456),(Item=(Type=Bone,Name="pelvis"),Weight=0.5516))
    Weight=1
   )
   > &RigUnit_ParentConstraint(
    AdvancedSettings=(RotationOrderForFilter=XZY)
    Child=(Type=Bone,Name="base_bk_skirt_l")
    Parents="((Item=(Type=Bone,Name=\"thigh_twist_01_r\"),Weight=0.0003),(Item=(Type=Bone,Name=\"deform_glute_r\"),Weight=0.0009),(Item=(Type=Bone,Name=\"thigh_twist_01_l\"),Weight=0.1003),(Item=(Type=Bone,Name=\"deform_glute_l\"),Weight=0.3417),(Item=(Type=Bone,Name=\"pelvis\") ...(truncated)"
    Weight=1
   )
   > &RigUnit_ParentConstraint(
    AdvancedSettings=(RotationOrderForFilter=XZY)
    Child=(Type=Bone,Name="base_belt_support_back_r")
    Parents="((Item=(Type=Bone,Name=\"dyn_belt_ammo_1\"),Weight=0.005),(Item=(Type=Bone,Name=\"spine_01\"),Weight=0.0084),(Item=(Type=Bone,Name=\"spine_02\"),Weight=0.0213),(Item=(Type=Bone,Name=\"pelvis\"),Weight=0.0862),(Item=(Type=Bone,Name=\"belt_center_r\"),Weight=0.1394),( ...(truncated)"
    Weight=1
   )
   > &RigUnit_ParentConstraint(
    AdvancedSettings=(RotationOrderForFilter=XZY)
    Child=(Type=Bone,Name="belt_support_front_r")
    Parents=((Item=(Type=Bone,Name="spine_02"),Weight=0.0066),(Item=(Type=Bone,Name="thigh_twist_01_r"),Weight=0.1453),(Item=(Type=Bone,Name="deform_groin_r"),Weight=0.1726),(Item=(Type=Bone,Name="pelvis"),Weight=0.6755))
    Weight=1
   )
   > &RigUnit_ParentConstraint(
    AdvancedSettings=(RotationOrderForFilter=XZY)
    Child=(Type=Bone,Name="base_belt_support_left")
    Parents=((Item=(Type=Bone,Name="thigh_twist_01_l"),Weight=0.0051),(Item=(Type=Bone,Name="spine_03"),Weight=0.0265),(Item=(Type=Bone,Name="spine_02"),Weight=0.1427),(Item=(Type=Bone,Name="spine_01"),Weight=0.3255),(Item=(Type=Bone,Name="pelvis"),Weight=0.5002))
    Weight=1
   )
   > &RigUnit_ParentConstraint(
    AdvancedSettings=(RotationOrderForFilter=XZY)
    Child=(Type=Bone,Name="base_belt_support_front_l")
    Parents="((Item=(Type=Bone,Name=\"thigh_twist_01_l\"),Weight=0.0007),(Item=(Type=Bone,Name=\"deform_groin_l\"),Weight=0.0062),(Item=(Type=Bone,Name=\"spine_03\"),Weight=0.0394),(Item=(Type=Bone,Name=\"spine_01\"),Weight=0.138),(Item=(Type=Bone,Name=\"spine_02\"),Weight=0.188 ...(truncated)"
    Weight=1
   )
   > &RigUnit_ParentConstraint(
    AdvancedSettings=(RotationOrderForFilter=XZY)
    Child=(Type=Bone,Name="base_jacket_back_c")
    Parents="((Item=(Type=Bone,Name=\"spine_01\"),Weight=0.0053),(Item=(Type=Bone,Name=\"spine_03\"),Weight=0.0478),(Item=(Type=Bone,Name=\"spine_02\"),Weight=0.0478),(Item=(Type=Bone,Name=\"clavicle_l\"),Weight=0.1285),(Item=(Type=Bone,Name=\"clavicle_r\"),Weight=0.1334),(Item= ...(truncated)"
    Weight=1
   )
   > &RigUnit_ParentConstraint(
    AdvancedSettings=(RotationOrderForFilter=XZY)
    Child=(Type=Bone,Name="base_jacket_fr_r")
    Parents="((Item=(Type=Bone,Name=\"deform_upperarm_fr_r\"),Weight=0.0003),(Item=(Type=Bone,Name=\"upperarm_twist_01_r\"),Weight=0.0021),(Item=(Type=Bone,Name=\"clavicle_r\"),Weight=0.0618),(Item=(Type=Bone,Name=\"spine_05\"),Weight=0.2813),(Item=(Type=Bone,Name=\"spine_04\"), ...(truncated)"
    Weight=1
   )
   > &RigUnit_ParentConstraint(
    AdvancedSettings=(RotationOrderForFilter=XZY)
    Child=(Type=Bone,Name="base_jacket_fr_l")
    Parents="((Item=(Type=Bone,Name=\"deform_upperarm_fr_l\"),Weight=0.0003),(Item=(Type=Bone,Name=\"upperarm_twist_01_l\"),Weight=0.0019),(Item=(Type=Bone,Name=\"clavicle_l\"),Weight=0.0568),(Item=(Type=Bone,Name=\"dyn_jacket_fr_l\"),Weight=0.0809),(Item=(Type=Bone,Name=\"spin ...(truncated)"
    Weight=1
   )
   RigVMFunctionReturnNode(ExecuteContext=@RigVMFunction_Sequence.B)
  ]}
 RigVMFunctionLibrary {nodes [
   Belt and Ribbon = RigVMCollapseNode(ExecuteContext=())
   Constraints = RigVMCollapseNode(ExecuteContext=())
   Pads_rotations = RigVMCollapseNode(ExecuteContext=())
  ]}
 RigVMFunction_Sequence_1_ContainedGraph {nodes [
   n1 = RigVMFunctionEntryNode()
   RigVMFunction_Sequence = &RigVMFunction_Sequence(ExecuteContext=@n1.ExecuteContext)
   n2 = &RigVMFunction_Sequence(ExecuteContext=@RigVMFunction_Sequence.B)
   n3 = &RigVMFunction_Sequence(ExecuteContext=@n2.B)
   RigVMFunctionReturnNode(A=@RigVMFunction_Sequence.A, B=@n2.A, C=@n3.A, D=@n3.B)
  ]}
 RigVMModel {nodes [
   n1 = &RigUnit_PrepareForExecution()
   &RigUnit_HierarchyImportFromSkeleton(NameSpace=None, bIncludeCurves=False, bIncludeMeshSockets=False, bIncludeVirtualBones=True, ExecutePin=@n1.ExecutePin)
   RigUnit_BeginExecution = &RigUnit_BeginExecution()
   FN_DeformRig_FNC_1 = RigVMFunctionReferenceNode(
    DeformRig_ Curve="(deformRig_l_upperarm_up_alpha_72_#1=False,deformRig_r_upperarm_up_alpha_73_#2=False,deformRig_l_upperarm_fr_alpha_74_#3=False,deformRig_r_upperarm_fr_alpha_75_CB51B8 ...(truncated)"
    DeformRig_Bool="(upperarm_1_#4=True,elbow_3_#5=True,wrist_7_#6=True,groin_8_#7=True,glute_17_#8=True,knee_9_40C408324CA261 ...(truncated)"
    DeformRig_Debug=(upperarm_debug_18_#4=False,wrist_debug_20_#6=False,lat_debug_26_#9=False,groinGlute_debug_27_#7=False)
    DeformRig_Value="(upperarm_up_18_#10=(Rotation=(X=0,Y=0,Z=0,W=1),Translation=(X=0,Y=0,Z=8),Scale3D=(X=0.7,Y=1,Z=1)),upperarm_fwd_fr_19_#11=(Rotation=(X=0,Y=0,Z=0,W=1),Translation=(X=0,Y=10,Z=0),Scale3D=(X=0.5,Y=1,Z= ...(truncated)"
    ExecuteContext=@RigUnit_BeginExecution.ExecutePin
   )
   > Constraints = RigVMFunctionReferenceNode()
   > Pads_rotations = RigVMFunctionReferenceNode()
   > Belt and Ribbon = RigVMFunctionReferenceNode()
   > n2 = &RigUnit_ModifyTransforms(
    ItemToModify="((Item=(Type=Bone,Name=\"jnt_helper_spine_04\"),Transform=(Rotation=(X=0,Y=0,Z=0,W=1),Translation=(X=0,Y=0,Z=0),Scale3D=(X=0.7,Y=0.7,Z=0.7))),(Item=(Type=Bone,Name=\"base_belt_support_left\"),Transform=(Rotation=(X=0,Y=0,Z=0,W=1),Translation=(X=0,Y=0,Z=1),Scal ...(truncated)"
    Weight=1
    WeightMinimum=0
   )
   RigVMFunction_ControlFlowBranch = &RigVMFunction_ControlFlowBranch(ExecuteContext=@n2.ExecutePin, Condition=RigVMVariableNode() {variable=isFrontEnd}.Value)
   &RigUnit_ModifyTransforms(
    ItemToModify="((Item=(Type=Bone,Name=\"base_jacket_fr_r\"),Transform=(Rotation=(X=0,Y=-0.0872,Z=0,W=0.9962),Translation=(X=0,Y=0,Z=0),Scale3D=(X=1,Y=1,Z=1))),(Item=(Type=Bone,Name=\"base_jacket_fr_l\"),Transform=(Rotation=(X=0,Y=-0.0872,Z=0,W=0.9962),Translation=(X=0,Y=0,Z= ...(truncated)"
    Weight=1
    WeightMinimum=0
    ExecutePin=@RigVMFunction_ControlFlowBranch.True
   )
  ]}
]
hierarchy [thigh_twist_01_r_ctrl {controlType=EULER_TRANSFORM, type=Control, loc=<0 0 0>, rot=<0 0 0 1>}]
variables [
 Leg_r_Swing_rot {type=FVector, default=(X=0,Y=0,Z=0)}
 Spine_Rotation {type=FVector, default=(X=0,Y=0,Z=0)}
 isCroching {type=bool, default=False}
 isFrontEnd {type=bool, default=False}
 isMovingCrouching {type=double, default=0}
 isPlayingEmotes {type=bool, default=False}
]
