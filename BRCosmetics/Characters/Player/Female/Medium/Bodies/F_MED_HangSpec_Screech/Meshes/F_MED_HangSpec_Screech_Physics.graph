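# Physics asset for the F_MED_HangSpec_Screech female player body cosmetic, set up for two short dangling ribbon tape strands.
# Left and right two-capsule chains hang from a small base_ribbon_tape body, with twist locked and 45-degree swing limits at each joint.
# The base joins spine_02 through a constraint with free swing and twist.
# Capsules on pelvis, spine_01 and spine_02 give the torso volume around the strands.
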PhysicsAsset /BRCosmetics/Characters/Player/Female/Medium/Bodies/F_MED_HangSpec_Screech/Meshes/F_MED_HangSpec_Screech_Physics
bodies {
 base_ribbon_tape [{shapeType=Capsule, length=0.505, radius=0.505, shapeName=base_ribbon_tape_capsule, rotation=<rot 0 -90 0>}]
 dyn_ribbon_tape_l [{shapeType=Capsule, length=5, radius=1, center=<5 0 0.8>, rotation=<rot -90 0 0>}]
 dyn_ribbon_tape_l_1 [{shapeType=Capsule, length=5, radius=1, center=<5 0 0.2>, rotation=<rot -90 0 0>}]
 dyn_ribbon_tape_r [{shapeType=Capsule, length=5, radius=1, center=<5 0 0.8>, rotation=<rot -90 0 0>}]
 dyn_ribbon_tape_r_1 [{shapeType=Capsule, length=5, radius=1.2, center=<5 0 0>, rotation=<rot -90 0 0>}]
 pelvis [{shapeType=Capsule, length=9, radius=9.7, shapeName=pelvis_capsule, center=<0 -1.2351 0>}]
 spine_01 [{shapeType=Capsule, length=7, radius=8, shapeName=spine_01_capsule}]
 spine_02 [{shapeType=Capsule, length=7.1854, radius=7, shapeName=spine_02_capsule, center=<0 0.7075 0>}]
}
constraints [
 {
  bone1Name=dyn_ribbon_tape_r_1
  bone2Name=dyn_ribbon_tape_r
  swing1LimitDegrees=45
  swing2LimitDegrees=45
  twistLimitDegrees=45
  twistMotion=Locked
 }
 {
  bone1Name=dyn_ribbon_tape_r
  bone2Name=base_ribbon_tape
  swing1LimitDegrees=45
  swing2LimitDegrees=45
  twistLimitDegrees=45
  twistMotion=Locked
 }
 {
  bone1Name=dyn_ribbon_tape_l_1
  bone2Name=dyn_ribbon_tape_l
  swing1LimitDegrees=45
  swing2LimitDegrees=45
  twistLimitDegrees=45
  twistMotion=Locked
 }
 {
  bone1Name=dyn_ribbon_tape_l
  bone2Name=base_ribbon_tape
  swing1LimitDegrees=45
  swing2LimitDegrees=45
  twistLimitDegrees=45
  twistMotion=Locked
 }
 {
  bone1Name=base_ribbon_tape
  bone2Name=spine_02
  swing1LimitDegrees=45
  swing1Motion=Free
  swing2LimitDegrees=45
  swing2Motion=Free
  twistLimitDegrees=45
  twistMotion=Free
 }
]
