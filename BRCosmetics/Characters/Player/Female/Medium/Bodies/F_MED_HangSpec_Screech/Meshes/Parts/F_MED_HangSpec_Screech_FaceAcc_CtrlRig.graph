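# Control Rig for the Screech female character's FaceAcc part, correcting hair bones as the head moves and during skydiving.
# A spherical pose reader on the head bone weights rotation offsets on base hair bones, then a branch on the skydiving bool rotates the bottom knot bone.
# Two further hair offset nodes and a controls function reference, apparently for copying head_ctrl onto the head bone, sit unwired in the main graph.
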ControlRigBlueprint /BRCosmetics/Characters/Player/Female/Medium/Bodies/F_MED_HangSpec_Screech/Meshes/Parts/F_MED_HangSpec_Screech_FaceAcc_CtrlRig
graphs [
 New Function_ContainedGraph {nodes [
   RigVMFunctionEntryNode()
   > &RigUnit_SetTransform(
    Item=(Type=Bone,Name="head")
    Weight=1
    bInitial=False
    Value=&RigUnit_GetTransform(Item=(Type=Control,Name="head_ctrl"), bInitial=False).Transform
   )
   > RigVMFunctionReturnNode()
  ]}
 RigVMFunctionLibrary {nodes [controls = RigVMCollapseNode(ExecuteContext=())]}
 RigVMModel {nodes [
   RigUnit_BeginExecution = &RigUnit_BeginExecution()
   RigVMFunction_Sequence = &RigVMFunction_Sequence(ExecuteContext=@RigUnit_BeginExecution.ExecutePin)
   n1 = &RigUnit_SphericalPoseReader(
    ActiveRegionScaleFactors=(PositiveWidth=1,NegativeWidth=0.2,PositiveHeight=1.5,NegativeHeight=1.5)
    ActiveRegionSize=0.1
    Debug=(bDrawDebug=false,bDraw2D=false,bDrawLocalAxes=false,DebugScale=25,DebugSegments=20,DebugThickness=0.25)
    DriverAxis=(X=0,Y=1,Z=0)
    DriverItem=(Type=Bone,Name="head")
    FalloffRegionScaleFactors=(PositiveWidth=3,NegativeWidth=1.8,PositiveHeight=2.5,NegativeHeight=2.5)
    FalloffSize=0.2
    FlipHeightScaling=False
    FlipWidthScaling=False
    OptionalParentItem=(Type=Bone,Name="None")
    RotationOffset=(X=90,Y=0,Z=-55)
    ExecutePin=@RigVMFunction_Sequence.B
   )
   > n2 = &RigUnit_ModifyTransforms(
    ItemToModify="((Item=(Type=Bone,Name=\"base_hair_c\"),Transform=(Rotation=(X=0,Y=-0.866,Z=0,W=0.5),Translation=(X=0,Y=0,Z=0),Scale3D=(X=1,Y=1,Z=1))),(Item=(Type=Bone,Name=\"base_hair_l\"),Transform=(Rotation=(X=0,Y=-0.866,Z=0,W=0.5),Translation=(X=0,Y=0,Z=0),Scale3D=(X=1,Y= ...(truncated)"
    WeightMinimum=0
    Weight=@n1.OutputParam
   )
   RigVMFunction_ControlFlowBranch = &RigVMFunction_ControlFlowBranch(ExecuteContext=@n2.ExecutePin, Condition=RigVMVariableNode() {variable=skydiving}.Value)
   &RigUnit_ModifyTransforms(
    ItemToModify=((Item=(Type=Bone,Name="dyn_knot_bottom"),Transform=(Rotation=(X=0,Y=-0.3827,Z=0,W=0.9239),Translation=(X=0,Y=0,Z=0),Scale3D=(X=1,Y=1,Z=1))))
    Weight=1
    WeightMinimum=0
    ExecutePin=@RigVMFunction_ControlFlowBranch.True
   )
   &RigUnit_ModifyTransforms(
    ExecutePin=()
    ItemToModify=((Item=(Type=Bone,Name="dyn_hair_top_01"),Transform=(Rotation=(X=0,Y=-0.1736,Z=0,W=0.9848),Translation=(X=0,Y=0,Z=0),Scale3D=(X=1,Y=1,Z=1))))
    Weight=1
    WeightMinimum=0
   )
   &RigUnit_ModifyTransforms(
    ExecutePin=()
    ItemToModify="((Item=(Type=Bone,Name=\"base_hair_c\"),Transform=(Rotation=(X=0,Y=-0.1736,Z=0,W=0.9848),Translation=(X=0,Y=0,Z=0),Scale3D=(X=1,Y=1,Z=1))),(Item=(Type=Bone,Name=\"base_hair_l\"),Transform=(Rotation=(X=0,Y=-0.2588,Z=0,W=0.9659),Translation=(X=0,Y=0,Z=0),Scale3D ...(truncated)"
    Weight=1
    WeightMinimum=0
   )
   controls = RigVMFunctionReferenceNode(ExecuteContext=())
  ]}
]
hierarchy [
 root {loc=<0 0 0>, rot=<0 0 0 1>}
 pelvis {parent=root, loc=<0 -0.5301 98.0571>, rot=<0.0127 0.707 -0.0127 -0.707>}
 spine_01 {parent=pelvis, loc=<6.1208 -0.9328 0>, rot=<0 0 0.034 -0.9994>}
 spine_02 {parent=spine_01, loc=<6.5997 -0.2095 0>, rot=<0 0 0.0238 -0.9997>}
 spine_03 {parent=spine_02, loc=<6.4914 -0.1237 0>, rot=<0 0 0.0149 -0.9999>}
 spine_04 {parent=spine_03, loc=<6.9598 0.0794 0>, rot=<0 0 -0.0132 -0.9999>}
 spine_05 {parent=spine_04, loc=<6.9987 0 0>, rot=<0 0 -0.0015 -1>}
 neck_01 {parent=spine_05, loc=<11.3138 0.1529 0>, rot=<0 0 -0.2003 -0.9797>}
 neck_02 {parent=neck_01, loc=<4.5627 0 0>, rot=<0 0 0.0504 -0.9987>}
 head {parent=neck_02, loc=<5.3914 0 0>, rot=<0 0 0.115 -0.9934>}
 base_hair_r {parent=head, loc=<2.1212 -4.6251 5.8191>, rot=<0.1453 -0.692 0.692 -0.1453>}
 dyn_hair_r_01 {parent=base_hair_r, loc=<0 0 0>, rot=<0 0 0 1>}
 dyn_hair_r_02 {parent=dyn_hair_r_01, loc=<6.3469 0 0>, rot=<0 0 0 1>}
 dyn_hair_r_03 {parent=dyn_hair_r_02, loc=<5.8685 0 0>, rot=<0 0 0 1>}
 dyn_hair_r_04 {parent=dyn_hair_r_03, loc=<5.8685 0 0>, rot=<0 -0.198 0 -0.9802>}
 dyn_hair_r_05 {parent=dyn_hair_r_04, loc=<5.8685 0 0>, rot=<0 0 0 1>}
 base_hair_l {parent=head, loc=<2.1211 -4.6251 -5.8191>, rot=<0.1453 -0.692 0.692 -0.1453>}
 dyn_hair_l_01 {parent=base_hair_l, loc=<0 0 0>, rot=<0 0 0 1>}
 dyn_hair_l_02 {parent=dyn_hair_l_01, loc=<6.347 0 0>, rot=<0 0 0 1>}
 dyn_hair_l_03 {parent=dyn_hair_l_02, loc=<5.8683 0 0.0002>, rot=<0 0 0 1>}
 dyn_hair_l_04 {parent=dyn_hair_l_03, loc=<5.8682 0 0.0001>, rot=<0 -0.198 0 -0.9802>}
 dyn_hair_l_05 {parent=dyn_hair_l_04, loc=<5.8691 0 -0.0001>, rot=<0 0 0 1>}
 base_hair_c {parent=head, loc=<2.7439 -6.0427 0>, rot=<0.1453 -0.692 0.692 -0.1453>}
 dyn_hair_c_01 {parent=base_hair_c, loc=<0 0 0>, rot=<0 0 0 1>}
 dyn_hair_c_02 {parent=dyn_hair_c_01, loc=<6.3469 0 0>, rot=<0 0 0 1>}
 dyn_hair_c_03 {parent=dyn_hair_c_02, loc=<7.5222 0 0>, rot=<0 -0.081 0 -0.9967>}
 dyn_hair_c_04 {parent=dyn_hair_c_03, loc=<6.6642 0 0.2196>, rot=<0 -0.1441 0 -0.9896>}
 dyn_hair_c_05 {parent=dyn_hair_c_04, loc=<7.5222 0 0>, rot=<0 0 0 1>}
 base_hair_top {parent=head, loc=<11.5001 -7.6646 0>, rot=<0.0902 -0.7013 0.7013 -0.0902>}
 dyn_hair_top_01 {parent=base_hair_top, loc=<0 0 0>, rot=<0 0 0 1>}
 dyn_hair_top_02 {parent=dyn_hair_top_01, loc=<4.651 0 0>, rot=<0 0 0 1>}
 dyn_hair_top_03 {parent=dyn_hair_top_02, loc=<5.8639 0 0>, rot=<0 0 0 1>}
 dyn_hair_top_04 {parent=dyn_hair_top_03, loc=<5.4865 0 0>, rot=<0 0 0 1>}
 dyn_knot_bottom {parent=base_hair_top, loc=<1.1357 0 1.6182>, rot=<0 0.1032 0 -0.9947>}
 base_hair_front_l {parent=head, loc=<4.6557 4.6759 -6.2011>, rot=<0.0041 -0.9999 -0.0001 -0.0158>}
 dyn_hair_front_l_01 {parent=base_hair_front_l, loc=<0 0 0>, rot=<0 0 0 1>}
 dyn_hair_front_l_02 {parent=dyn_hair_front_l_01, loc=<4.4041 0 0>, rot=<0 0 0 1>}
 dyn_hair_front_l_03 {parent=dyn_hair_front_l_02, loc=<4.4041 0 0>, rot=<0 0 0 1>}
 base_hair_front_r {parent=head, loc=<4.6553 4.6759 6.2011>, rot=<0.0158 0.0001 0.9999 -0.0041>}
 dyn_hair_front_r_01 {parent=base_hair_front_r, loc=<0 0 0>, rot=<0 0 0 1>}
 dyn_hair_front_r_02 {parent=dyn_hair_front_r_01, loc=<4.4032 0 0>, rot=<0 0 0 1>}
 dyn_hair_front_r_03 {parent=dyn_hair_front_r_02, loc=<4.4042 0 0>, rot=<0 0 0 1>}
 base_knot_top {parent=head, loc=<12.3608 -8.4256 0>, rot=<0.6051 -0.3658 0.3658 -0.6051>}
 dyn_knot_top {parent=base_knot_top, loc=<0 0 0>, rot=<0 0 0 1>}
 clavicle_l {parent=spine_05, loc=<5.4831 -0.1053 -3.2946>, rot=<0.0465 -0.7289 -0.009 -0.6829>}
 upperarm_l {parent=clavicle_l, loc=<9.8658 -1.2136 -0.4105>, rot=<0.0164 -0.3245 -0.0269 -0.9454>}
 clavicle_r {parent=spine_05, loc=<5.4831 -0.1053 3.2946>, rot=<0.7289 0.0465 -0.6829 0.009>}
 upperarm_r {parent=clavicle_r, loc=<-9.8658 1.2136 0.4105>, rot=<0.0164 -0.3245 -0.0269 -0.9454>}
 head_ctrl {controlType=EULER_TRANSFORM, type=Control, loc=<0 0 0>, rot=<0 0 0 1>}
]
variables [skydiving {type=bool, default=False}]
