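# Physics asset for the F_MED_HangSpec_Screech female face accessory mesh, defining capsule bodies and joints for dynamic hair.
# Center, left, right, two front and top chains hang from base bones jointed to the head, mostly with swing limits; front chains lock twist.
# Free-rotation cross joints tie the lower side segments to the center chain.
# Head, neck, spine, clavicle and upper-arm capsules give the hair body shapes to collide with.
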PhysicsAsset /BRCosmetics/Characters/Player/Female/Medium/Bodies/F_MED_HangSpec_Screech/Meshes/Parts/F_MED_HangSpec_Screech_FaceAcc_Physics
bodies {
 base_hair_c [{shapeType=Capsule, length=0.505, radius=0.505, shapeName=base_hair_c_capsule, rotation=<rot 0 -90 0>}]
 base_hair_front_l [{shapeType=Capsule, length=0.505, radius=0.505, shapeName=base_hair_front_l_capsule, rotation=<rot 0 -90 0>}]
 base_hair_front_r [{shapeType=Capsule, length=0.505, radius=0.505, shapeName=base_hair_front_r_capsule, rotation=<rot 0 -90 0>}]
 base_hair_l [{shapeType=Capsule, length=0.505, radius=0.505, shapeName=base_hair_l_capsule, rotation=<rot 0 -90 0>}]
 base_hair_r [{shapeType=Capsule, length=0.505, radius=0.505, shapeName=base_hair_r_capsule, rotation=<rot 0 -90 0>}]
 base_hair_top [{shapeType=Capsule, length=1, radius=1, center=<1 0 0>, rotation=<rot -90 0 0>}]
 clavicle_l [{shapeType=Capsule, length=5.24, radius=6.5, shapeName=clavicle_l_capsule, center=<4 -1 0>, rotation=<rot -90 0 0>}]
 clavicle_r [{shapeType=Capsule, length=5.24, radius=6.5, shapeName=clavicle_r_capsule, center=<-4 1 0>, rotation=<rot -90 0 0>}]
 dyn_hair_c_01 [{shapeType=Capsule, length=10, radius=2.2, center=<3 0 2>, rotation=<rot -90 -90 0>}]
 dyn_hair_c_02 [{shapeType=Capsule, length=10, radius=2.2, center=<3 0 2>, rotation=<rot -90 -90 0>}]
 dyn_hair_c_03 [{shapeType=Capsule, length=10, radius=2.2, center=<3 0 1>, rotation=<rot -90 -90 0>}]
 dyn_hair_c_04 [{shapeType=Capsule, length=4, radius=2.2, center=<3 0 0.9908>, rotation=<rot -90 0 0>}]
 dyn_hair_c_05 [{shapeType=Capsule, length=4, radius=3, center=<3 0 0.9908>, rotation=<rot -90 0 0>}]
 dyn_hair_front_l_01 [{shapeType=Capsule, length=4, radius=2, center=<3 0 1.2>, rotation=<rot -90 0 0>}]
 dyn_hair_front_l_02 [{shapeType=Capsule, length=4, radius=2, center=<3 0 1.2>, rotation=<rot -90 0 0>}]
 dyn_hair_front_l_03 [{shapeType=Capsule, length=4, radius=2, center=<3 0 0.2>, rotation=<rot -90 0 0>}]
 dyn_hair_front_r_01 [{shapeType=Capsule, length=4, radius=2, center=<3 0 1.5>, rotation=<rot -90 0 0>}]
 dyn_hair_front_r_02 [{shapeType=Capsule, length=4, radius=2, center=<3 0 0.5>, rotation=<rot -90 0 0>}]
 dyn_hair_front_r_03 [{shapeType=Capsule, length=4, radius=2, center=<3 0 0.2>, rotation=<rot -90 0 0>}]
 dyn_hair_l_01 [{shapeType=Capsule, length=4, radius=2.5, center=<6 2 0>, rotation=<rot -90 0 0>}]
 dyn_hair_l_02 [{shapeType=Capsule, length=6, radius=3, center=<6 3 0.5068>, rotation=<rot -90 90 0>}]
 dyn_hair_l_03 [{shapeType=Capsule, length=6, radius=3, center=<6 3 0.5068>, rotation=<rot -90 90 0>}]
 dyn_hair_l_04 [{shapeType=Capsule, length=6, radius=2, center=<6 3 0>, rotation=<rot -90 90 0>}]
 dyn_hair_l_05 [{shapeType=Capsule, length=6, radius=2, center=<6 2 0>, rotation=<rot -90 90 0>}]
 dyn_hair_r_01 [{shapeType=Capsule, length=4, radius=2.5, center=<6 -2 0>, rotation=<rot -90 0 0>}]
 dyn_hair_r_02 [{shapeType=Capsule, length=6, radius=2.5, center=<6 -3 0>, rotation=<rot -90 90 0>}]
 dyn_hair_r_03 [{shapeType=Capsule, length=6, radius=2.5, center=<6 -3 0>, rotation=<rot -90 90 0>}]
 dyn_hair_r_04 [{shapeType=Capsule, length=6, radius=2, center=<6 -3 0>, rotation=<rot -90 90 0>}]
 dyn_hair_r_05 [{shapeType=Capsule, length=6, radius=2, center=<6 -3 0>, rotation=<rot -90 -90 0>}]
 dyn_hair_top_01 [{shapeType=Capsule, length=1, radius=1, center=<1 0 0>, rotation=<rot -90 0 0>}]
 dyn_hair_top_02 [{shapeType=Capsule, length=3, radius=2, center=<6 0 3>, rotation=<rot -90 0 0>}]
 dyn_hair_top_03 [{shapeType=Capsule, length=3, radius=2, center=<6 0 3>, rotation=<rot -90 0 0>}]
 dyn_hair_top_04 [{shapeType=Capsule, length=3, radius=2, center=<6 0 3>, rotation=<rot -90 0 0>}]
 head [{shapeType=Capsule, length=6.155, radius=6.155, shapeName=head_capsule, center=<1.083 1.2155 0>, rotation=<rot -90 0 0>}]
 neck_01 [{shapeType=Capsule, length=5, radius=4, shapeName=neck_01_capsule, rotation=<rot -90 0 0>}]
 neck_02 [{shapeType=Capsule, length=5, radius=4, shapeName=neck_02_capsule, rotation=<rot -90 0 0>}]
 spine_03 [{shapeType=Capsule, length=9.7387, radius=2, shapeName=spine_04_capsule, center=<2.8313 0 0>}]
 spine_04 [{shapeType=Capsule, length=9.7387, radius=9, shapeName=spine_04_capsule, center=<2.8313 0 0>}]
 spine_05 [{shapeType=Capsule, length=9.2271, radius=9, shapeName=spine_05_capsule, center=<2.7524 1 0>}]
 upperarm_l [{shapeType=Capsule, length=20, radius=5, shapeName=upperarm_l_capsule, center=<10 0 0>, rotation=<rot -90 0 0>}]
 upperarm_r [{shapeType=Capsule, length=5.1967, radius=5, shapeName=upperarm_r_capsule, center=<-2 0 0>, rotation=<rot -90 0 0>}]
}
constraints [
 {bone1Name=dyn_hair_c_02, bone2Name=dyn_hair_c_01, swing1LimitDegrees=15, swing2LimitDegrees=45, twistLimitDegrees=2}
 {bone1Name=dyn_hair_c_03, bone2Name=dyn_hair_c_02, swing1LimitDegrees=15, swing2LimitDegrees=45, twistLimitDegrees=2}
 {bone1Name=dyn_hair_c_04, bone2Name=dyn_hair_c_03, swing1LimitDegrees=15, swing2LimitDegrees=45, twistLimitDegrees=2}
 {bone1Name=dyn_hair_c_05, bone2Name=dyn_hair_c_04, swing1LimitDegrees=15, swing2LimitDegrees=45, twistLimitDegrees=2}
 {
  bone1Name=dyn_hair_front_l_01
  bone2Name=base_hair_front_l
  swing1LimitDegrees=25
  swing2LimitDegrees=25
  twistLimitDegrees=45
  twistMotion=Locked
 }
 {
  bone1Name=dyn_hair_front_l_02
  bone2Name=dyn_hair_front_l_01
  swing1LimitDegrees=25
  swing2LimitDegrees=25
  twistLimitDegrees=45
  twistMotion=Locked
 }
 {
  bone1Name=dyn_hair_front_l_03
  bone2Name=dyn_hair_front_l_02
  swing1LimitDegrees=25
  swing2LimitDegrees=25
  twistLimitDegrees=45
  twistMotion=Locked
 }
 {bone1Name=dyn_hair_l_01, bone2Name=base_hair_l, swing1LimitDegrees=15, swing2LimitDegrees=45, twistLimitDegrees=2}
 {bone1Name=dyn_hair_l_02, bone2Name=dyn_hair_l_01, swing1LimitDegrees=15, swing2LimitDegrees=45, twistLimitDegrees=2}
 {bone1Name=dyn_hair_l_03, bone2Name=dyn_hair_l_02, swing1LimitDegrees=15, swing2LimitDegrees=45, twistLimitDegrees=2}
 {bone1Name=dyn_hair_l_04, bone2Name=dyn_hair_l_03, swing1LimitDegrees=15, swing2LimitDegrees=45, twistLimitDegrees=2}
 {bone1Name=dyn_hair_l_05, bone2Name=dyn_hair_l_04, swing1LimitDegrees=15, swing2LimitDegrees=45, twistLimitDegrees=2}
 {bone1Name=dyn_hair_r_01, bone2Name=base_hair_r, swing1LimitDegrees=15, swing2LimitDegrees=45, twistLimitDegrees=2}
 {bone1Name=dyn_hair_r_02, bone2Name=dyn_hair_r_01, swing1LimitDegrees=15, swing2LimitDegrees=45, twistLimitDegrees=2}
 {bone1Name=dyn_hair_r_03, bone2Name=dyn_hair_r_02, swing1LimitDegrees=15, swing2LimitDegrees=45, twistLimitDegrees=2}
 {bone1Name=dyn_hair_r_04, bone2Name=dyn_hair_r_03, swing1LimitDegrees=15, swing2LimitDegrees=45, twistLimitDegrees=2}
 {bone1Name=dyn_hair_r_05, bone2Name=dyn_hair_r_04, swing1LimitDegrees=15, swing2LimitDegrees=45, twistLimitDegrees=2}
 {
  bone1Name=base_hair_front_l
  bone2Name=head
  swing1LimitDegrees=45
  swing1Motion=Free
  swing2LimitDegrees=45
  swing2Motion=Free
  twistLimitDegrees=45
  twistMotion=Free
 }
 {bone1Name=dyn_hair_c_01, bone2Name=base_hair_c, swing1LimitDegrees=15, swing2LimitDegrees=45, twistLimitDegrees=2}
 {
  bone1Name=base_hair_c
  bone2Name=head
  swing1LimitDegrees=45
  swing1Motion=Free
  swing2LimitDegrees=45
  swing2Motion=Free
  twistLimitDegrees=45
  twistMotion=Free
 }
 {
  bone1Name=base_hair_l
  bone2Name=head
  swing1LimitDegrees=45
  swing1Motion=Free
  swing2LimitDegrees=45
  swing2Motion=Free
  twistLimitDegrees=45
  twistMotion=Free
 }
 {
  bone1Name=base_hair_r
  bone2Name=head
  swing1LimitDegrees=45
  swing1Motion=Free
  swing2LimitDegrees=45
  swing2Motion=Free
  twistLimitDegrees=45
  twistMotion=Free
 }
 {
  bone1Name=dyn_hair_r_04
  bone2Name=dyn_hair_c_04
  swing1LimitDegrees=45
  swing1Motion=Free
  swing2LimitDegrees=45
  swing2Motion=Free
  twistLimitDegrees=45
  twistMotion=Free
 }
 {
  bone1Name=dyn_hair_l_04
  bone2Name=dyn_hair_c_04
  swing1LimitDegrees=45
  swing1Motion=Free
  swing2LimitDegrees=45
  swing2Motion=Free
  twistLimitDegrees=45
  twistMotion=Free
 }
 {
  bone1Name=dyn_hair_r_05
  bone2Name=dyn_hair_c_05
  swing1LimitDegrees=45
  swing1Motion=Free
  swing2LimitDegrees=45
  swing2Motion=Free
  twistLimitDegrees=45
  twistMotion=Free
 }
 {
  bone1Name=dyn_hair_l_05
  bone2Name=dyn_hair_c_05
  swing1LimitDegrees=45
  swing1Motion=Free
  swing2LimitDegrees=45
  swing2Motion=Free
  twistLimitDegrees=45
  twistMotion=Free
 }
 {
  bone1Name=dyn_hair_front_r_03
  bone2Name=dyn_hair_front_r_02
  swing1LimitDegrees=25
  swing2LimitDegrees=25
  twistLimitDegrees=45
  twistMotion=Locked
 }
 {
  bone1Name=base_hair_front_r
  bone2Name=head
  swing1LimitDegrees=45
  swing1Motion=Free
  swing2LimitDegrees=45
  swing2Motion=Free
  twistLimitDegrees=45
  twistMotion=Free
 }
 {
  bone1Name=dyn_hair_front_r_01
  bone2Name=base_hair_front_r
  swing1LimitDegrees=25
  swing2LimitDegrees=25
  twistLimitDegrees=45
  twistMotion=Locked
 }
 {
  bone1Name=dyn_hair_front_r_02
  bone2Name=dyn_hair_front_r_01
  swing1LimitDegrees=25
  swing2LimitDegrees=25
  twistLimitDegrees=45
  twistMotion=Locked
 }
 {
  bone1Name=base_hair_top
  bone2Name=head
  swing1LimitDegrees=45
  swing1Motion=Free
  swing2LimitDegrees=45
  swing2Motion=Free
  twistLimitDegrees=45
  twistMotion=Free
 }
 {
  bone1Name=dyn_hair_top_01
  bone2Name=base_hair_top
  swing1LimitDegrees=45
  swing1Motion=Free
  swing2LimitDegrees=45
  swing2Motion=Free
  twistLimitDegrees=45
  twistMotion=Free
 }
 {bone1Name=dyn_hair_top_02, bone2Name=dyn_hair_top_01, swing1LimitDegrees=5, swing2LimitDegrees=35, twistLimitDegrees=45}
 {bone1Name=dyn_hair_top_03, bone2Name=dyn_hair_top_02, swing1LimitDegrees=5, swing2LimitDegrees=60, twistLimitDegrees=45}
 {bone1Name=dyn_hair_top_04, bone2Name=dyn_hair_top_03, swing1LimitDegrees=5, swing2LimitDegrees=60, twistLimitDegrees=45}
]
