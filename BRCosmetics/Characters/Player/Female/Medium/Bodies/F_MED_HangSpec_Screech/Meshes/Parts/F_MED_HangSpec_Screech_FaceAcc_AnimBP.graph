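# Post-physics AnimBP for the hair bones of the Screech female character's FaceAcc part.
# It copies the parent mesh pose through a skydiving-aware Control Rig into a RigidBody sim with state-driven gravity, then caches it.
# When backbling or a cape is worn, or physics is off, constraints tie left, right and center hair chains to body bones.
# Either branch gets its own base_hair_top tilt, plus a shared skydiving-dependent rotation.
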AnimBlueprint /BRCosmetics/Characters/Player/Female/Medium/Bodies/F_MED_HangSpec_Screech/Meshes/Parts/F_MED_HangSpec_Screech_FaceAcc_AnimBP
target_skeleton=&/F_MED_HangSpec_Screech_FaceAcc_Skeleton
graphs [
 AnimGraph {nodes [
   n1 = "Constraint - dyn_hair_r_01 (PP)"!(
    ComponentPose=Local To Component(LocalPose=Use cached pose 'ctrlRig_Cache'())
    Alpha=1
    ConstraintWeights_0=0.5
    ConstraintWeights_1:float=0.5
   ) -> (Pose) {
    alphaBoolBlend {blendOption=Linear}
    boneToModify {boneName=dyn_hair_r_01}
    constraintSetup [{targetBone {boneName=head}}, {targetBone {boneName=neck_02}}]
    constraintWeights [1,1]
   }
   n2 = "Constraint - dyn_hair_r_02 (PP)"!(ComponentPose=@n1, Alpha=1, ConstraintWeights_0=1, ConstraintWeights_1:float=0) -> (Pose) {
    alphaBoolBlend {blendOption=Linear}
    boneToModify {boneName=dyn_hair_r_02}
    constraintSetup [{targetBone {boneName=clavicle_r}}, {targetBone {boneName=neck_02}}]
    constraintWeights [1,1]
   }
   n3 = "Constraint - dyn_hair_r_03 (PPP)"!(ComponentPose=@n2, Alpha=1, ConstraintWeights_0=0, ConstraintWeights_1:float=0, ConstraintWeights_2:float=1) -> (Pose) {
    alphaBoolBlend {blendOption=Linear}
    boneToModify {boneName=dyn_hair_r_03}
    constraintSetup [{targetBone {boneName=neck_01}}, {targetBone {boneName=spine_05}}, {targetBone {boneName=clavicle_r}}]
    constraintWeights [1,1,1]
   }
   n4 = "Constraint - dyn_hair_r_04 (PP)"!(ComponentPose=@n3, Alpha=1, ConstraintWeights_0=0.5, ConstraintWeights_1:float=0.5) -> (Pose) {
    alphaBoolBlend {blendOption=Linear}
    boneToModify {boneName=dyn_hair_r_04}
    constraintSetup [{targetBone {boneName=spine_05}}, {targetBone {boneName=clavicle_r}}]
    constraintWeights [1,1]
   }
   n5 = "Constraint - dyn_hair_r_05 (PP)"!(ComponentPose=@n4, Alpha=1, ConstraintWeights_0=0.5, ConstraintWeights_1:float=0.5) -> (Pose) {
    alphaBoolBlend {blendOption=Linear}
    boneToModify {boneName=dyn_hair_r_05}
    constraintSetup [{targetBone {boneName=spine_04}}, {targetBone {boneName=spine_05}}]
    constraintWeights [1,1]
   }
   n6 = "Constraint - dyn_hair_l_01 (PP)"!(ComponentPose=@n5, Alpha=1, ConstraintWeights_0=0.5, ConstraintWeights_1:float=0.5) -> (Pose) {
    alphaBoolBlend {blendOption=Linear}
    boneToModify {boneName=dyn_hair_l_01}
    constraintSetup [{targetBone {boneName=head}}, {targetBone {boneName=neck_02}}]
    constraintWeights [1,1]
   }
   n7 = "Constraint - dyn_hair_l_02 (PP)"!(ComponentPose=@n6, Alpha=1, ConstraintWeights_0=1, ConstraintWeights_1:float=0) -> (Pose) {
    alphaBoolBlend {blendOption=Linear}
    boneToModify {boneName=dyn_hair_l_02}
    constraintSetup [{targetBone {boneName=clavicle_l}}, {targetBone {boneName=neck_02}}]
    constraintWeights [1,1]
   }
   n8 = "Constraint - dyn_hair_l_03 (PPP)"!(ComponentPose=@n7, Alpha=1, ConstraintWeights_0=0, ConstraintWeights_1:float=0, ConstraintWeights_2:float=1) -> (Pose) {
    alphaBoolBlend {blendOption=Linear}
    boneToModify {boneName=dyn_hair_l_03}
    constraintSetup [{targetBone {boneName=neck_01}}, {targetBone {boneName=spine_05}}, {targetBone {boneName=clavicle_l}}]
    constraintWeights [1,1,1]
   }
   n9 = "Constraint - dyn_hair_l_04 (P)"(ComponentPose=@n8, Alpha=1, ConstraintWeights_0=1) {
    alphaBoolBlend {blendOption=Linear}
    boneToModify {boneName=dyn_hair_l_04}
    constraintSetup [{targetBone {boneName=spine_05}}]
    constraintWeights [1]
   }
   n10 = "Constraint - dyn_hair_l_05 (PP)"!(ComponentPose=@n9, Alpha=1, ConstraintWeights_0=0.5, ConstraintWeights_1:float=0.5) -> (Pose) {
    alphaBoolBlend {blendOption=Linear}
    boneToModify {boneName=dyn_hair_l_05}
    constraintSetup [{targetBone {boneName=spine_04}}, {targetBone {boneName=spine_05}}]
    constraintWeights [1,1]
   }
   n11 = "Constraint - dyn_hair_c_01 (PP)"!(ComponentPose=@n10, Alpha=1, ConstraintWeights_0=0.5, ConstraintWeights_1:float=0.5) -> (Pose) {
    alphaBoolBlend {blendOption=Linear}
    boneToModify {boneName=dyn_hair_c_01}
    constraintSetup [{targetBone {boneName=head}}, {targetBone {boneName=neck_02}}]
    constraintWeights [1,1]
   }
   n12 = "Constraint - dyn_hair_c_02 (PP)"!(ComponentPose=@n11, Alpha=1, ConstraintWeights_0=0.5, ConstraintWeights_1:float=0.5) -> (Pose) {
    alphaBoolBlend {blendOption=Linear}
    boneToModify {boneName=dyn_hair_c_02}
    constraintSetup [{targetBone {boneName=neck_01}}, {targetBone {boneName=neck_02}}]
    constraintWeights [1,1]
   }
   n13 = "Constraint - dyn_hair_c_03 (PP)"!(ComponentPose=@n12, Alpha=1, ConstraintWeights_0=0.5, ConstraintWeights_1:float=0.5) -> (Pose) {
    alphaBoolBlend {blendOption=Linear}
    boneToModify {boneName=dyn_hair_c_03}
    constraintSetup [{targetBone {boneName=neck_01}}, {targetBone {boneName=spine_05}}]
    constraintWeights [1,1]
   }
   n14 = "Constraint - dyn_hair_c_04 (PP)"!(ComponentPose=@n13, Alpha=1, ConstraintWeights_0=0.5, ConstraintWeights_1:float=0.5) -> (Pose) {
    alphaBoolBlend {blendOption=Linear}
    boneToModify {boneName=dyn_hair_c_04}
    constraintSetup [{targetBone {boneName=spine_05}}, {targetBone {boneName=clavicle_l}}]
    constraintWeights [1,1]
   }
   n15 = "Constraint - dyn_hair_c_05 (PP)"!(ComponentPose=@n14, Alpha=1, ConstraintWeights_0=0.5, ConstraintWeights_1:float=0.5) -> (Pose) {
    alphaBoolBlend {blendOption=Linear}
    boneToModify {boneName=dyn_hair_c_05}
    constraintSetup [{targetBone {boneName=spine_04}}, {targetBone {boneName=spine_05}}]
    constraintWeights [1,1]
   }
   n16 = "Transform (Modify) Bone - Bone: base_hair_top"(ComponentPose=@n15, Alpha=1, Translation="0,0,0", Rotation="0,0,2.5", Scale="1,1,1") {
    rotationMode=BMM_Additive
    scaleMode=BMM_Additive
    scaleSpace=BCS_BoneSpace
    alphaBoolBlend {blendOption=Linear}
    boneToModify {boneName=base_hair_top}
    rotation {roll=2.5}
   }
   n17 = Component To Local(ComponentPose=@n16)
   n18 = "Transform (Modify) Bone - Bone: base_hair_top"(
    ComponentPose=Local To Component(LocalPose=Use cached pose 'ctrlRig_Cache'())
    Alpha=1
    Translation="0,0,0"
    Rotation="0,0,-3"
    Scale="1,1,1"
   ) {
    rotationMode=BMM_Additive
    scaleMode=BMM_Additive
    scaleSpace=BCS_BoneSpace
    alphaBoolBlend {blendOption=Linear}
    boneToModify {boneName=base_hair_top}
    rotation {roll=-3}
   }
   n19 = Component To Local(ComponentPose=@n18)
   n20 = OR Boolean!(A=Get isWearingBackbling(), B=Get Is Wearing Cape() -> (Is Wearing Cape:bool), C:bool=NOT Boolean(A=Get bRigidBodyEnabled())) -> (ReturnValue)
   n21 = "Transform (Modify) Bone - Bone: base_hair_top"!(
    ComponentPose=Local To Component(LocalPose=Blend Poses by bool(BlendPose_0=@n17, BlendPose_1=@n19, bActiveValue=@n20) {blendType=Linear})
    bAlphaBoolEnabled:bool=True
    Translation="0,0,0"
    Rotation=Get RotFalloff() -> (RotFalloff:Rotator)
    Scale="1,1,1"
   ) -> (Pose) {
    alphaInputType=Bool
    rotationMode=BMM_Additive
    rotationSpace=BCS_BoneSpace
    alphaBoolBlend {blendInTime=0.2, blendOption=Linear, blendOutTime=0.2}
    boneToModify {boneName=base_hair_top}
   }
   Output Pose(Result=Component To Local(ComponentPose=@n21)) {layerGroup=None}
   n22 = Control Rig!(Source=Copy Pose From Mesh() {bUseAttachedParent=true}, Alpha, skydiving:bool=Get bIsSkyDiving()) -> (Pose) {
    alphaBoolBlend {blendOption=Linear}
    controlRigAssetReference {blueprintRigClass=&/F_MED_HangSpec_Screech_FaceAcc_CtrlRig.F_MED_HangSpec_Screech_FaceAcc_CtrlRig_C}
   }
   n23 = RigidBody!(
    ComponentPose=Local To Component(LocalPose=@n22)
    Alpha
    OverrideWorldGravity:Vector=Get GravityOverride()
    ExternalForce
    SimSpaceSettings:Sim Space Settings Structure=(WorldAlpha=0,VelocityScaleZ=1,DampingAlpha=1,MaxLinearVelocity=10000,MaxAngularVelocity=10000,MaxLinearAcceleration=10000,MaxAngularAcceleration=10000,ExternalLinearDragV=(X=0,Y=0,Z=0),ExternalLinearVelocity=(X=0,Y=0,Z=0),ExternalAngularVelocity=(X=0,Y=0,Z=0))
   ) -> (Pose) {
    bClampLinearTranslationLimitToRefPose=true
    bFreezeIncomingPoseOnStart=true
    bOverrideWorldGravity=true
    bUseExternalClothCollision=true
    overridePhysicsAsset=&/F_MED_HangSpec_Screech_FaceAcc_Physics
    simulationSpace=BaseBoneSpace
    alphaBoolBlend {blendOption=Linear}
    baseBoneRef {boneName=neck_01}
    componentAppliedLinearAccClamp=<0 0 0>
    overrideWorldGravity {z=-980}
    simSpaceSettings {maxAngularAcceleration=100, maxAngularVelocity=4, maxLinearAcceleration=700, maxLinearVelocity=300, velocityScaleZ=0.5, worldAlpha=1}
   }
   Save cached pose 'ctrlRig_Cache'(Pose=Component To Local(ComponentPose=@n23)) {cacheName=ctrlRig_Cache}
  ]}
 EventGraph {nodes [
   Event BlueprintUpdateAnimation()
   > n1 = Sequence()
   > n2 = Gravity Override State Function DEPRECATED(
    CharacterPartAnimInstance=Self-Reference()
    GameTimeInSeconds=GetGameTimeInSeconds()
    GravityOverrideParams=Get Gravity Override Params()
    RBAN Params=Get RBAN Params()
   )
   > Set GravityOverride(GravityOverride=@n2)
   > Set RBAN Settings Out(RBAN Settings Out:RBANSetting Struct Structure=@n2.RBAN SettingsOut) -> (Output_Get:RBANSetting Struct Structure)
   > n3 = Branch(Condition=Get bIsSkyDiving())
   > Set RotFalloff(RotFalloff:Rotator="15,0,0") -> (Output_Get:Rotator)
   Set RotFalloff(execute=@n3.else, RotFalloff="5,0,0")
   n4 = CheckForBackbling()
   Set isWearingBackbling(
    execute=@n1.then_1
    isWearingBackbling=AND Boolean!(A=@n4, B=NOT Boolean(A=@n4.bIsBackblingACape), C:bool=NOT Boolean(A=@n4.bIsBackblingLong)) -> (ReturnValue)
   )
   > Set Is Wearing Cape(Is Wearing Cape=OR Boolean!(A=@n4.bIsBackblingACape, B=@n4.bIsBackblingLong) -> (ReturnValue))
  ]}
]
variables [
 Gravity Override Params {type=GravityOverrideParamsStruct, default {skydive_GoingUp {gravityOverride=<0 -300 100>}}}
 GravityOverride {type=Vector}
 Is Wearing Cape {type=boolean}
 isWearingBackbling {type=boolean}
 RBAN Params {
  type=RBANParamsStruct
  default {
   dBNO {simSpaceSettings {
     dampingAlpha=0
     maxAngularAcceleration=32
     maxAngularVelocity=8
     maxLinearAcceleration=2800
     maxLinearVelocity=700
     velocityScaleZ=0.5
     externalLinearDragV=<0.5 0.5 0.5>
    }}
   dBNO_Carried {simSpaceSettings {
     dampingAlpha=0
     maxAngularAcceleration=32
     maxAngularVelocity=8
     maxLinearAcceleration=2800
     maxLinearVelocity=700
     velocityScaleZ=0.5
     externalLinearDragV=<0.5 0.5 0.5>
    }}
   emoteOrMelee {simSpaceSettings {
     dampingAlpha=0
     maxAngularAcceleration=32
     maxAngularVelocity=8
     maxLinearAcceleration=2800
     maxLinearVelocity=700
     velocityScaleZ=0.5
     externalLinearDragV=<0.5 0.5 0.5>
    }}
   front_End {simSpaceSettings {
     dampingAlpha=0
     maxAngularAcceleration=10
     maxAngularVelocity=2.5
     maxLinearAcceleration=800
     maxLinearVelocity=200
     velocityScaleZ=0.5
     externalLinearDragV=<0 0 0>
    }}
   onGround_Crouch_Moving {simSpaceSettings {
     dampingAlpha=0
     maxAngularAcceleration=32
     maxAngularVelocity=8
     maxLinearAcceleration=3200
     maxLinearVelocity=400
     velocityScaleZ=0.5
     externalLinearDragV=<0.5 0.5 0.5>
    }}
   onGround_Falling {simSpaceSettings {
     dampingAlpha=0
     maxAngularAcceleration=32
     maxAngularVelocity=8
     maxLinearAcceleration=2800
     maxLinearVelocity=700
     velocityScaleZ=0.3
     externalLinearDragV=<0 0 0>
    }}
   onGround_Moving {simSpaceSettings {
     dampingAlpha=0
     maxAngularAcceleration=24
     maxAngularVelocity=6
     maxLinearAcceleration=1200
     maxLinearVelocity=300
     velocityScaleZ=0.5
     externalLinearDragV=<0.2 0.2 0.2>
    }}
   onGround_Standing {simSpaceSettings {
     dampingAlpha=0
     maxAngularAcceleration=24
     maxAngularVelocity=6
     maxLinearAcceleration=1200
     maxLinearVelocity=300
     velocityScaleZ=0.5
     externalLinearDragV=<0.2 0.2 0.2>
    }}
   rail_Grinding {simSpaceSettings {
     dampingAlpha=0
     maxAngularAcceleration=32
     maxAngularVelocity=8
     maxLinearAcceleration=2800
     maxLinearVelocity=700
     velocityScaleZ=0.5
     worldAlpha=1
     externalLinearDragV=<0.5 0.5 0.5>
    }}
   riding_Moving {simSpaceSettings {
     dampingAlpha=0
     maxAngularAcceleration=32
     maxAngularVelocity=8
     maxLinearAcceleration=2800
     maxLinearVelocity=700
     velocityScaleZ=0.5
     externalLinearDragV=<0.5 0.5 0.5>
    }}
   riding_Moving_Targeting {simSpaceSettings {
     dampingAlpha=0
     maxAngularAcceleration=32
     maxAngularVelocity=8
     maxLinearAcceleration=2800
     maxLinearVelocity=700
     velocityScaleZ=0.5
     externalLinearDragV=<0.5 0.5 0.5>
    }}
   riding_Standing {simSpaceSettings {
     dampingAlpha=0
     maxAngularAcceleration=32
     maxAngularVelocity=8
     maxLinearAcceleration=2800
     maxLinearVelocity=700
     velocityScaleZ=0.5
     externalLinearDragV=<0.5 0.5 0.5>
    }}
   skydive_DivingDown {simSpaceSettings {
     dampingAlpha=0
     maxAngularAcceleration=32
     maxAngularVelocity=8
     maxLinearAcceleration=2800
     maxLinearVelocity=700
     velocityScaleZ=0.5
     externalLinearDragV=<0.5 0.5 0.5>
    }}
   skydive_FloatingDown {
    componentLinearAccScale=<0 0.5 0>
    componentLinearVelScale=<0 0.5 0>
    simSpaceSettings {
     dampingAlpha=0
     maxAngularAcceleration=32
     maxAngularVelocity=8
     maxLinearAcceleration=2800
     maxLinearVelocity=700
     velocityScaleZ=0.5
     externalLinearDragV=<0.5 0.5 0.5>
    }
   }
   skydive_GoingUp {
    componentLinearAccScale=<0 -0.5 0>
    componentLinearVelScale=<0 -0.5 0>
    simSpaceSettings {
     dampingAlpha=0
     maxAngularAcceleration=16
     maxAngularVelocity=4
     maxLinearAcceleration=1200
     maxLinearVelocity=300
     velocityScaleZ=0
     externalLinearDragV=<0 0 -1>
    }
   }
   skydive_Parachuting {
    componentLinearAccScale=<0 0.1 0.3>
    componentLinearVelScale=<0 0.1 0.3>
    simSpaceSettings {
     dampingAlpha=0
     maxAngularAcceleration=32
     maxAngularVelocity=8
     maxLinearAcceleration=2800
     maxLinearVelocity=700
     velocityScaleZ=0.5
     externalLinearDragV=<0.5 0.5 0.5>
    }
   }
   skydive_Surfing {simSpaceSettings {
     dampingAlpha=0
     maxAngularAcceleration=32
     maxAngularVelocity=8
     maxLinearAcceleration=2800
     maxLinearVelocity=700
     velocityScaleZ=0.5
     externalLinearDragV=<0.5 0.5 0.5>
    }}
   swimming {simSpaceSettings {
     dampingAlpha=0
     maxAngularAcceleration=32
     maxAngularVelocity=8
     maxLinearAcceleration=2800
     maxLinearVelocity=700
     velocityScaleZ=0.5
     externalLinearDragV=<0.5 0.5 0.5>
    }}
   tactical_Sprint {simSpaceSettings {
     dampingAlpha=0
     maxAngularAcceleration=24
     maxAngularVelocity=6
     maxLinearAcceleration=2000
     maxLinearVelocity=500
     velocityScaleZ=0.25
     externalLinearDragV=<0.2 0.2 0.2>
    }}
   ziplining {simSpaceSettings {
     dampingAlpha=0
     maxAngularAcceleration=32
     maxAngularVelocity=8
     maxLinearAcceleration=2800
     maxLinearVelocity=700
     velocityScaleZ=0.5
     externalLinearDragV=<0.5 0.5 0.5>
    }}
  }
 }
 RBAN Settings Out {
  type=RBANSettingStruct
  default {
   componentLinearAccScale=<0 0 0>
   componentLinearVelScale=<0 0 0>
   simSpaceSettings {
    dampingAlpha=0
    maxAngularAcceleration=16
    maxAngularVelocity=4
    maxLinearAcceleration=800
    maxLinearVelocity=200
    velocityScaleZ=0.5
    worldAlpha=1
    externalAngularVelocity=<0 0 0>
    externalLinearDragV=<0 0 0>
    externalLinearVelocity=<0 0 0>
   }
  }
 }
 RotFalloff {type=Rotator}
]
