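# Material function that outputs a 0-1 coordinate within the mesh's local bounding box, as RGB plus separate R, G and B channels.
# It transforms world position to instance space, subtracts the local bounds minimum and divides by the bounds size.
# A parallax branch offsets world position along the camera-to-pixel vector by a depth input, but its result is unconnected.
# A local-position vertex interpolator is also left unused.
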MaterialFunction /BRCosmetics/Characters/Player/Female/Medium/Bodies/F_MED_HangSpec_Screech/Skins/MF_BoundingBoxBased_01_UVW_Parallax
graph {
 properties {bExposeToLibrary=true, description=Creates a locally aligned 0-1 value on XY and Z., libraryCategoriesText [Misc,Coordinates]}
 nodes [
  n1 = WorldPosition()
  n2 = Subtract(A=@n1.XYZ, B=TransformPosition(None=CameraPositionWS()) {transformSourceType=TRANSFORMPOSSOURCE_World})
  n3 = ScalarParameter() {defaultValue=2, parameterName=Parallax Depth}
  Add(
   A=@n1.XYZ
   B=Multiply(A=Normalize(VectorInput=@n2), B=FunctionInput(Preview=@n3) {bUsePreviewValueAsDefault=true, inputType=FunctionInput_Scalar})
  )
  n4 = MaterialFunctionCall() {materialFunction=&ObjectLocalBounds}
  n5 = Divide(
   A=Subtract(
    A=TransformPosition(None=@n1.XYZ) {transformSourceType=TRANSFORMPOSSOURCE_World, transformType=TRANSFORMPOSSOURCE_Instance}
    B=@n4.Local Bounds Minimum
   )
   B=@n4.Local Bounds Size
  )
  FunctionOutput(None=@n5) {outputName=RGB}
  n6 = MaterialFunctionCall(Float3=@n5) {materialFunction=&BreakOutFloat3Components}
  FunctionOutput(None=@n6.R) {outputName=R, sortPriority=1}
  FunctionOutput(None=@n6.G) {outputName=G, sortPriority=2}
  FunctionOutput(None=@n6.B) {outputName=B, sortPriority=3}
  VertexInterpolator(VS=LocalPosition().XYZ)
 ]
}
