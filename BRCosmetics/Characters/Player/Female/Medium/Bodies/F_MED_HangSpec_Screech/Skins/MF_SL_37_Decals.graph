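# Material function for decals on a Screech female character skin, layering a masked, view-dependent emissive color onto incoming material attributes.
# An inverted Fresnel with adjustable exponent multiplies the input color, so the glow is strongest on camera-facing surfaces and fades toward grazing angles.
# The result overrides emissive in a copy of the attributes, which is blended with the original using mask times intensity as alpha.
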MaterialFunction /BRCosmetics/Characters/Player/Female/Medium/Bodies/F_MED_HangSpec_Screech/Skins/MF_SL_37_Decals
graph {nodes [
  Constant()
  n1 = FunctionInput() {inputName=Material Attributes, inputType=FunctionInput_MaterialAttributes}
  n2 = GetMaterialAttributes(None=@n1) {attributeGetTypes [#1]}
  n3 = FunctionInput() {inputName=Fresnel Exp, inputType=FunctionInput_Scalar, sortPriority=20}
  n4 = Multiply(A=FunctionInput() {inputName=Color, sortPriority=10}, B=OneMinus(None=Fresnel(ExponentIn=@n3)))
  n5 = FunctionInput() {inputName=Mask, inputType=FunctionInput_Scalar, sortPriority=30}
  n6 = FunctionInput() {inputName=Intensity, inputType=FunctionInput_Scalar, sortPriority=40}
  FunctionOutput(None=BlendMaterialAttributes(
    A=@n1
    B=SetMaterialAttributes!(MaterialAttributes=@n2.MaterialAttributes, BaseColor=@n2.MaterialAttributes, EmissiveColor=@n4, Metallic) {attributeSetTypes [#1, #2, #3]}
    Alpha=Multiply(A=@n5, B=@n6)
   )) {outputName=BlendMaterialAttributes}
 ]}
