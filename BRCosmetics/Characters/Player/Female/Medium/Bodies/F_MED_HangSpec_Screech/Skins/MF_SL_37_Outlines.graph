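# Material function for the SL 37 skin of F_MED_HangSpec_Screech that turns masked outlines into glowing, view-dependent emissive lines.
# An inverted Fresnel, floored and combined with a smoothstepped highlight, scales outline color and strength before adding to incoming emissive.
# A contrasted "Buckle" Fresnel, tinted and gated by Additional Mask, is max-blended in.
# The Outlines mask lerps in this emissive and darkens base color to near black.
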MaterialFunction /BRCosmetics/Characters/Player/Female/Medium/Bodies/F_MED_HangSpec_Screech/Skins/MF_SL_37_Outlines
graph {nodes [
  n1 = GetMaterialAttributes(None=FunctionInput() {inputName="MA (In)", inputType=FunctionInput_MaterialAttributes}) {attributeGetTypes [#1, #2]}
  n2 = VectorParameter() {group=SL 37 - Outlines, parameterName=Outline - Emissive Color, defaultValue=<rgba 0 0.4218 0.6172 1>}
  n3 = FunctionInput() {inputName=Outlines, inputType=FunctionInput_Scalar}
  n4 = Multiply(
   A=ScalarParameter() {defaultValue=1.5, group=SL 37 - Outlines, parameterName=Outline - Strength}
   B=Multiply(A=FunctionInput(Preview=@n2.RGB) {bUsePreviewValueAsDefault=true, inputName=Outline Color}, B=@n3)
  )
  n5 = Fresnel(ExponentIn=ScalarParameter() {defaultValue=0.4, group=SL 37 - Outlines, parameterName=Outline - Fresnel}) {baseReflectFraction=0}
  n6 = ScalarParameter() {group=SL 37 - Outlines, parameterName=Outline - Highlight Step}
  n7 = ScalarParameter() {group=SL 37 - Outlines, parameterName=Outline - Highlight Smoothness}
  n8 = Fresnel(ExponentIn=ScalarParameter() {group=SL 37 - Outlines, parameterName=Outline - Highlight Fresnel}) {baseReflectFraction=0}
  n9 = Max(
   A=Max(A=OneMinus(None=@n5), B=ScalarParameter() {defaultValue=0.15, group=SL 37 - Outlines, parameterName=Outline - Fresnel Min})
   B=SmoothStep(Min=@n6, Max=Add(A=@n6, B=@n7), Value=OneMinus(None=@n8))
  )
  n10 = MaterialFunctionCall(
   In=Fresnel(ExponentIn=ScalarParameter() {defaultValue=1.6233, group=SL 37 - Outlines, parameterName=Buckle - Exp})
   Contrast=ScalarParameter() {defaultValue=0.25, group=SL 37 - Outlines, parameterName=Buckle - Contrast}
  ) {materialFunction=&CheapContrast}
  n11 = Max(
   A=Add(A=@n1.EmissiveColor, B=Multiply(A=@n4, B=@n9))
   B=Multiply(A=FunctionInput() {inputName=Additional Mask, inputType=FunctionInput_Scalar}, B=Multiply(A=@n2.RGB, B=@n10.Result))
  )
  FunctionOutput(None=SetMaterialAttributes!(
    MaterialAttributes=@n1.MaterialAttributes
    EmissiveColor=LinearInterpolate(A=@n1.EmissiveColor, B=@n11, Alpha=@n3)
    BaseColor=LinearInterpolate(A=@n1.BaseColor, Alpha=@n3) {constB=0.01}
   ) {attributeSetTypes [#1, #2]}) {outputName=SetMaterialAttributes}
 ]}
