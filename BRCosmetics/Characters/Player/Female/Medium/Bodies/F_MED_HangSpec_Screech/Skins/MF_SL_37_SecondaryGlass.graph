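# Material function that adds a glowing "SL 37" glass look to areas of the Screech character skin selected by the Secondary input.
# It blends incoming attributes toward MF_TransparentArmor, then adds tinted Fresnel/V-gradient emissive, maxed against a time-scrolling band along pre-skinned local Z.
# Use Glass bypasses everything when off.
# UseVertexHide optionally pulls vertices toward the object position through World Position Offset, unmasked by Secondary.
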MaterialFunction /BRCosmetics/Characters/Player/Female/Medium/Bodies/F_MED_HangSpec_Screech/Skins/MF_SL_37_SecondaryGlass
graph {nodes [
  n1 = FunctionInput() {inputName="MA (In)", inputType=FunctionInput_MaterialAttributes, sortPriority=10}
  n2 = FunctionInput() {inputName=Secondary, inputType=FunctionInput_Scalar}
  n3 = GetMaterialAttributes(None=BlendMaterialAttributes(
    A=@n1
    B=MaterialFunctionCall() {materialFunction=/BRCosmetics/Characters/Player/Female/Medium/Bodies/F_MED_HangSpec_Screech/Materials/MF_TransparentArmor}.Result
    Alpha=@n2
   )) {attributeGetTypes [#1]}
  n4 = Fresnel(ExponentIn=ScalarParameter() {defaultValue=1.272, group=SL 37 - Glass, parameterName=Glass Color - Fresnel Exp}) {baseReflectFraction=0}
  n5 = Add(
   A=Power(
    Base=MaterialFunctionCall() {materialFunction=&LinearGradient}.VGradient
    Exp=ScalarParameter() {defaultValue=5, group=SL 37 - Glass, parameterName=Glass - Gradient Power}
   )
   B=ScalarParameter() {defaultValue=-0.4, group=SL 37 - Glass, parameterName=Glass - Gradient}
  )
  n6 = Multiply(
   A=VectorParameter() {group=SL 37 - Glass, parameterName=Glass Color, defaultValue=<rgba 0 0.6982 0.8385 1>}.RGB
   B=ScalarParameter() {defaultValue=1, group=SL 37 - Glass, parameterName=Glass Color - Strength}
  )
  n7 = Add(A=@n3.EmissiveColor, B=Multiply(A=Max(A=@n4, B=@n5), B=@n6))
  n8 = Constant()
  n9 = LocalPosition() {localOrigin=InstancePreSkinning}
  n10 = Add(
   A=Multiply(A=VertexInterpolator(VS=@n9.Z).PS, B=ScalarParameter() {defaultValue=0.2, group=SL 37 - Glass, parameterName=Glass - Band Amount})
   B=Multiply(A=Time(), B=ScalarParameter() {defaultValue=0.25, group=SL 37 - Glass, parameterName=Glass - Band Speed})
  )
  n11 = Multiply(
   A=Multiply(A=@n6, B=ScalarParameter() {defaultValue=3, group=SL 37 - Glass, parameterName=Glass- Band Strength})
   B=MaterialFunctionCall!(
    Width=ScalarParameter() {defaultValue=0.1, group=SL 37 - Glass, parameterName=Glass - Band Width}
    Sharpness=ScalarParameter() {defaultValue=0.1, group=SL 37 - Glass, parameterName=Glass - Band Sharpness}
    Offset
    Direction Switch
    Compare
    Input Coordinates=MaterialFunctionCall(X=@n8, Y=@n10) {materialFunction=&MakeFloat2}.Result
   ) {materialFunction=/Engine/Functions/Engine_MaterialFunctions03/Procedurals/GeneratedBand}.Result
  )
  n12 = BlendMaterialAttributes(
   A=@n3.MaterialAttributes
   B=SetMaterialAttributes!(MaterialAttributes=@n3.MaterialAttributes, EmissiveColor=Max(A=@n7, B=@n11)) {attributeSetTypes [#1]}
   Alpha=@n2
  )
  n13 = ObjectPositionWS()
  n14 = WorldPosition()
  n15 = SetMaterialAttributes!(
   MaterialAttributes=@n12
   WorldPositionOffset=Multiply(A=Subtract(A=@n13, B=@n14.XYZ), B=ScalarParameter() {defaultValue=1, group=00_Glass, parameterName=VertexHide, sortPriority=301})
  ) {attributeSetTypes [#2]}
  FunctionOutput(None=StaticSwitchParameter(True=StaticSwitchParameter(True=@n15, False=@n12) {group=00_Glass, parameterName=UseVertexHide, sortPriority=300}, False=@n1) {group=SL 37 - Glass, parameterName=Use Glass}) {outputName="Switch Param (False)", sortPriority=10}
  WorldPosition()
 ]}
