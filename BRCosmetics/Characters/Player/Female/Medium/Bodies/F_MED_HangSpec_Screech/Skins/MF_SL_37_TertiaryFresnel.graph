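# Material function for the Screech SL 37 skin that applies a near-black, iridescent tertiary layer to incoming material attributes where the Tertiary mask is set.
# A sine-remapped Fresnel, flippable and noise-distorted, indexes a shared curve atlas for emissive color, switchable to flat color, plus a second Fresnel rim.
# The layer also blends in tiled skin-bump normals, sets metallic and offers optional clear coat.
# Some three-color and noise-Fresnel branches are left unconnected.
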MaterialFunction /BRCosmetics/Characters/Player/Female/Medium/Bodies/F_MED_HangSpec_Screech/Skins/MF_SL_37_TertiaryFresnel
graph {nodes [
  n1 = GetMaterialAttributes(None=FunctionInput() {inputName="MA (In)", inputType=FunctionInput_MaterialAttributes}) {attributeGetTypes [#1, #2, #3]}
  n2 = Saturate(None=FunctionInput() {inputName=Tertiary, inputType=FunctionInput_Scalar})
  n3 = LinearInterpolate(A=@n1.BaseColor, Alpha=@n2) {constB=0.009}
  n4 = ScalarParameter() {
   defaultValue=10
   desc="Metallic - Color Curve \r\n\r\nThis number allows you to choose which index you want to use in your Curve Atlas so you don't need to use multiple curve atlas for different color curves"
   group=SL 37 - Tertiary
   parameterName=Secondary - Curve Index
   sortPriority=30
  }
  n5 = ScalarParameter() {
   defaultValue=13
   desc="Color Curve \r\n\r\nThis is the total amount of color curves in your Curve Atlas"
   group=SL 37 - Tertiary
   parameterName=Secondary - Curve Amount
   sortPriority=29
  }
  n6 = Fresnel(ExponentIn=ScalarParameter() {defaultValue=0.6, group=SL 37 - Tertiary, parameterName=Tertiary - Fresnel Exp}, Normal=NamedRerouteUsage())
  n7 = ScalarParameter() {defaultValue=1, group=SL 37 - Tertiary, parameterName=Tertiary - Thin Film Amount}
  n8 = Multiply(A=MaterialFunctionCall!(Sine Phase=Multiply(A=@n6, B=@n7), Value 1, Value 2) {materialFunction=/Engine/Functions/Engine_MaterialFunctions02/Utility/Sine_Remapped}.Remapped Sine)
  n9 = LinearInterpolate(A=@n8, B=OneMinus(None=@n8), Alpha=FunctionInput() {inputName=Color Gradient Mask, inputType=FunctionInput_Scalar})
  n10 = VectorParameter() {group=SL 37 - Tertiary, parameterName=Tertiary - Color Curve Distortion - Tiling, defaultValue=<rgba 14.0354 3.1607 0 0>}
  n11 = Multiply(A=TextureCoordinate(), B=ComponentMask(None=@n10.RGB) {g=true, r=true})
  n12 = ScalarParameter() {group=SL 37 - Tertiary, parameterName=Tertiary - Color Curve Distortion}
  n13 = Add(
   A=LinearInterpolate(A=@n9, B=OneMinus(None=@n9), Alpha=ScalarParameter() {group=SL 37 - Tertiary, parameterName=Tertiary - Flip Fresnel})
   B=Multiply(
    A=TextureSample(UVs=@n11) {samplerType=SAMPLERTYPE_Grayscale, texture=&LowResBlurredNoise}.R
    B=FunctionInput(Preview=@n12) {bUsePreviewValueAsDefault=true, inputName=Distortion Amount, inputType=FunctionInput_Scalar}
   )
  )
  n14 = Saturate(None=MaterialFunctionCall!(
    CurveIndex=FunctionInput(Preview=@n4) {bUsePreviewValueAsDefault=true, inputName=Secondary - Curve Index, inputType=FunctionInput_Scalar}
    CurveNum=FunctionInput(Preview=@n5) {bUsePreviewValueAsDefault=true, inputName=Secondary - Curve Amount, inputType=FunctionInput_Scalar}
    GreyMapIn=Saturate(None=@n13)
   ) {materialFunction=/Game/Effects/Niagara/MasterMaterials/Functions/Common/FXCommon_CurveAtlasUV}.UV)
  n15 = TextureObjectParameter() {
   group=SL 37 - Global
   parameterName=Color Curves
   samplerType=SAMPLERTYPE_LinearColor
   texture=/BRCosmetics/Characters/Player/Female/Medium/Bodies/F_MED_HangSpec_Screech/Skins/SuperlevelT1/Textures/CA_SL37
  }
  n16 = VectorParameter() {group=SL 37 - Tertiary, parameterName=Tertiary Fresnel Color, defaultValue=<rgba 0.1119 0.7682 0.7682 1>}
  n17 = StaticSwitch(
   True=Multiply(
    A=TextureSample(UVs=@n14, Tex=@n15) {samplerType=SAMPLERTYPE_LinearColor}.RGB
    B=ScalarParameter() {defaultValue=1, group=SL 37 - Tertiary, parameterName=Tertiary - Color Curve Intensity}
   )
   False=@n16.RGB
   Value=FunctionInput(Preview=StaticBool() {value=true}) {bUsePreviewValueAsDefault=true, inputName=Use Color Curve, inputType=FunctionInput_StaticBool}
  )
  n18 = Fresnel(ExponentIn=ScalarParameter() {defaultValue=3, group=SL 37 - Tertiary, parameterName=Tertiary Fresnel Amount})
  n19 = FunctionInput(Preview=@n16.RGB) {bUsePreviewValueAsDefault=true, inputName=Tertiary Fresnel Color}
  n20 = StaticBool()
  n21 = Multiply(A=TextureCoordinate(), B=ScalarParameter() {defaultValue=15, parameterName=TilingTert})
  n22 = ScalarParameter() {group=SL 37 - Tertiary, parameterName=Bump Normal Flatten}
  n23 = MaterialFunctionCall(
   BaseNormal=@n1.Normal
   AdditionalNormal=MaterialFunctionCall(
    Normal=TextureSample(UVs=@n21) {samplerType=SAMPLERTYPE_Normal, texture=/Game/Global/Textures/Default/Tileable_Patterns/T_SkinBump_N}.RGB
    Flatness=FunctionInput(Preview=@n22) {bUsePreviewValueAsDefault=true, inputName=Normal Intensity, inputType=FunctionInput_Scalar}
   ) {materialFunction=&FlattenNormal}.Result
  ) {materialFunction=&BlendAngleCorrectedNormals}
  n24 = SetMaterialAttributes!(
   MaterialAttributes=@n1.MaterialAttributes
   BaseColor=@n3
   EmissiveColor=Add(A=LinearInterpolate(A=@n1.EmissiveColor, B=@n17, Alpha=@n2), B=Multiply(A=@n18, B=@n19))
   CustomData0=Constant() {r=1}
   CustomData1=ScalarParameter() {defaultValue=0.0943, group=SL 37 - Tertiary, parameterName="Clear  Coat - Roughness"}
   ShadingModel=StaticSwitch(
    True=ShadingModel() {shadingModel=MSM_ClearCoat}
    False=ShadingModel()
    Value=FunctionInput(Preview=@n20) {inputName=Use Clear Coat, inputType=FunctionInput_StaticBool}
   )
   Metallic=FunctionInput(Preview=ScalarParameter() {defaultValue=1, group=SL 37 - Tertiary, parameterName=Tertiary - Metallic}) {bUsePreviewValueAsDefault=true, inputName=Metallic, inputType=FunctionInput_Scalar}
   Normal=@n23.Result
   Roughness=Constant() {r=1}
   Specular=Constant() {r=0.5}
  ) {attributeSetTypes [
    #1
    #2
    #4
    #5
    #6
    #7
    #3
    #8
    #9
   ]}
  FunctionOutput(None=BlendMaterialAttributes(A=@n1.MaterialAttributes, B=@n24, Alpha=@n2)) {outputName="MA (Out)"}
  FunctionOutput(None=@n3) {outputName=Base Color, sortPriority=10}
  n25 = Add(
   A=Fresnel(ExponentIn=ScalarParameter() {defaultValue=0.8095, parameterName=ExponentIn})
   B=ScalarParameter() {defaultValue=4.0924, parameterName=Offset Thin}
  )
  MaterialFunctionCall(Alpha=MaterialFunctionCall(Sine Phase=Multiply(A=@n25) {constB=3}) {materialFunction=&Sine_Remapped}.Remapped Sine) {materialFunction=&Lerp_3Color}
  n26 = Fresnel(ExponentIn=ScalarParameter() {defaultValue=1.1484, parameterName=T3 - Fresnel Noise Amount})
  n27 = TextureSample() {samplerType=SAMPLERTYPE_Grayscale, texture=&LowResBlurredNoise}
  Max(A=@n18, B=Multiply(A=Multiply(A=@n26, B=@n27.R), B=ScalarParameter() {defaultValue=1, parameterName=T3 - Fresnel Noise}))
  NamedRerouteDeclaration(None=Transform(None=@n23.Result)) {name=PoreNRM, nodeColor=<rgba 1 0.2344 0 1>}
 ]}
