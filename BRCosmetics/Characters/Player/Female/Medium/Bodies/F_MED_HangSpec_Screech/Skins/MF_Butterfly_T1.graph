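# Material function for the T1 Butterfly superlevel skin on the Screech character body, restyling incoming attributes with paint, glow and effects.
# Non-skin areas get desaturated paint; the primary mask blends in a subsurface layer with a distorted, parallax triplanar butterfly pattern, atlas color curve and optional glitter.
# Secondary glass, tertiary fresnel, outline and decal functions are chained after it, driven by breakout and tech-art masks.
# Outputs include the result, painted base and masks.
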MaterialFunction /BRCosmetics/Characters/Player/Female/Medium/Bodies/F_MED_HangSpec_Screech/Skins/MF_Butterfly_T1
graph {nodes [
  Constant()
  Constant()
  Constant()
  DotProduct(A=Transform(), B=CameraVectorWS())
  n1 = TextureSampleParameter2D() {
   group=SL 37 - Global
   parameterName=Breakout Masks
   texture=/BRCosmetics/Characters/Player/Female/Medium/Bodies/F_MED_HangSpec_Screech/Skins/SuperlevelT1/Textures/T_Screech_Body_SL_Mask01
  }
  n2 = Step(X=@n1.A) {constX=0.4, constY=0.48}
  n3 = Step(X=@n1.A) {constX=0.5, constY=0.5}
  FunctionOutput(None=NamedRerouteDeclaration(None=Subtract(A=@n2, B=@n3)) {name=Additional Mask, nodeColor=<rgba 0.6406 1 0 1>}) {outputName=Additional Mask}
  n4 = NamedRerouteDeclaration(None=@n1.R) {name=Primary, nodeColor=<rgba 1 0 0 1>}
  n5 = NamedRerouteDeclaration(None=@n1.G) {name=Secondary, nodeColor=<rgba 0 1 0 1>}
  n6 = NamedRerouteDeclaration(None=@n1.B) {name=Tertiary, nodeColor=<rgba 0 0 1 1>}
  FunctionOutput(None=MaterialFunctionCall(X=@n4, Y=@n5, Z=@n6, A=NamedRerouteDeclaration(None=@n1.A) {name=Color Gradient Mask, nodeColor=<rgba 1 1 1 1>}) {materialFunction=&MakeFloat4}.Result) {outputName=SL Mask}
  n7 = NamedRerouteUsage()
  n8 = NamedRerouteUsage()
  n9 = VectorParameter() {group=SL37 - T1 Butterfly, parameterName=T1 - Outline Emissive Color, defaultValue=<rgba 0 1.0251 1.5 1>}
  n10 = FunctionInput() {inputName="MA (In)", inputType=FunctionInput_MaterialAttributes}
  n11 = GetMaterialAttributes(None=@n10) {attributeGetTypes [#1, #2]}
  n12 = Power(Base=Desaturation(None=@n11.BaseColor), Exp=ScalarParameter() {defaultValue=0.1, group=SL37 - T1 Butterfly, parameterName=T1 - Desat Pow})
  n13 = LinearInterpolate(
   A=ScalarParameter() {defaultValue=0.02, group=SL37 - T1 Butterfly, parameterName=T1 - Paint Strength}
   B=ScalarParameter() {defaultValue=0.2, group=SL37 - T2 Spider, parameterName=T1 - Metal Paint Strength}
   Alpha=@n11.Metallic
  )
  n14 = ScalarParameter() {defaultValue=1, group=SL37 - T1 Butterfly, parameterName=T1 - Metallic Multipler}
  n15 = Subtract(A=OneMinus(None=FunctionInput() {inputName=Skin Mask, inputType=FunctionInput_Scalar}), B=NamedRerouteUsage())
  n16 = BlendMaterialAttributes(
   A=@n10
   B=SetMaterialAttributes!(MaterialAttributes=@n11.MaterialAttributes, BaseColor=Multiply(A=@n12, B=@n13), Metallic=Multiply(A=@n11.Metallic, B=@n14)) {attributeSetTypes [#1, #2]}
   Alpha=@n15
  )
  n17 = Constant()
  n18 = BlendMaterialAttributes(
   A=@n16
   B=SetMaterialAttributes!(MaterialAttributes=GetMaterialAttributes(None=@n16).MaterialAttributes, EmissiveColor=@n17, BaseColor, SubsurfaceColor=@n17) {attributeSetTypes [#3, #1, #4]}
   Alpha=@n15
  )
  n19 = GetMaterialAttributes(None=@n18) {attributeGetTypes [
    #5
    #6
    #7
    #1
   ]}
  n20 = Power(Base=FunctionInput() {inputName=M, inputType=FunctionInput_Scalar}, Exp=ScalarParameter() {defaultValue=2, parameterName=Exp_1})
  n21 = TextureObject() {
   samplerType=SAMPLERTYPE_Masks
   texture=/BRCosmetics/Characters/Player/Female/Medium/Bodies/F_MED_HangSpec_Screech/Skins/SuperlevelT1/Textures/T_Butterfly_M_01
  }
  n22 = ScalarParameter() {defaultValue=64, group=SL37 - T1 Butterfly, parameterName=World X}
  n23 = ScalarParameter() {defaultValue=64, group=SL37 - T1 Butterfly, parameterName=World Y}
  n24 = ScalarParameter() {defaultValue=96, group=SL37 - T1 Butterfly, parameterName=World Z}
  n25 = VertexInterpolator(VS=LocalPosition() {localOrigin=InstancePreSkinning}.XYZ)
  n26 = TextureObject() {samplerType=SAMPLERTYPE_Grayscale, texture=&LowResBlurredNoise}
  n27 = ScalarParameter() {defaultValue=16, parameterName=Distortion Tiling}
  n28 = VectorParameter() {parameterName=Distortion Movement, defaultValue=<rgba 0 0 1 0>}
  n29 = Add(A=Multiply(A=Time(), B=@n28.RGB), B=@n25.PS)
  n30 = VertexInterpolator(VS=PreSkinnedNormal())
  n31 = Multiply(
   A=ComponentMask(None=MaterialFunctionCall(TextureObject=@n26, TextureSize=@n27, WorldPosition=@n29, World Space Normal=@n30.PS) {materialFunction=&WorldAlignedTexture}.XYZ Texture) {r=true}
   B=ScalarParameter() {defaultValue=1.75, parameterName=Distortion Scale}
  )
  n32 = ScalarParameter() {defaultValue=-4, group=SL37 - T1 Butterfly, parameterName=Butterfly Pattern - Depth}
  n33 = ComponentMask(None=MaterialFunctionCall!(
    TextureObject=@n21
    TextureSize=MaterialFunctionCall(X=@n22, Y=@n23, Z=@n24) {materialFunction=&MakeFloat3}.Result
    WorldPosition=Add(A=@n25.PS, B=@n31)
    Export Float 4
    Parallax=@n32
    In
    World Space Normal=@n30.PS
    ProjectionTransitionContrast=ScalarParameter() {defaultValue=0.2, group=SL37 - T1 Butterfly, parameterName=Triplanar Contrast}
   ) {materialFunction=/BRCosmetics/Characters/Player/Male/Medium/Bodies/M_MED_SilverBellMarine/Materials/MF_Triplanar_ScalePan_Parallax}.XYZ Texture) {r=true}
  n34 = ScalarParameter() {defaultValue=1, parameterName=Butterfly Pattern - Intensity}
  n35 = OneMinus(None=NamedRerouteDeclaration(None=Multiply(A=@n33, B=@n34)) {name=Butterfly Pattern, nodeColor=<rgba 1 0.2109 0 1>})
  n36 = ScalarParameter() {
   desc="Metallic - Color Curve \r\n\r\nThis number allows you to choose which index you want to use in your Curve Atlas so you don't need to use multiple curve atlas for different color curves"
   group=SL 37 - Global
   parameterName=Curve Index
   sortPriority=30
  }
  n37 = ScalarParameter() {
   defaultValue=13
   desc="Color Curve \r\n\r\nThis is the total amount of color curves in your Curve Atlas"
   group=SL 37 - Global
   parameterName=Curve Amount
   sortPriority=29
  }
  n38 = NamedRerouteUsage()
  n39 = TextureObjectParameter() {
   parameterName=Color Curves
   samplerType=SAMPLERTYPE_LinearColor
   texture=/BRCosmetics/Characters/Player/Female/Medium/Bodies/F_MED_HangSpec_Screech/Skins/SuperlevelT1/Textures/CA_SL37
  }
  n40 = TextureSample(
   UVs=Saturate(None=MaterialFunctionCall!(CurveIndex=@n36, CurveNum=@n37, GreyMapIn=@n38) {materialFunction=/Game/Effects/Niagara/MasterMaterials/Functions/Common/FXCommon_CurveAtlasUV}.UV)
   Tex=@n39
  ) {
   samplerSource=SSM_Wrap_WorldGroupSettings
   samplerType=SAMPLERTYPE_LinearColor
   texture=&CA_SL37
  }
  n41 = NamedRerouteDeclaration(None=@n40.RGB) {name=Color Curve, nodeColor=<rgba 0 0.4844 1 1>}
  n42 = ScalarParameter() {defaultValue=0.8187, parameterName=ExponentIn_5}
  n43 = Multiply(
   A=Multiply(A=@n41, B=ScalarParameter() {defaultValue=1, parameterName=B_4})
   B=OneMinus(None=Fresnel(ExponentIn=@n42) {baseReflectFraction=0})
  )
  n44 = Fresnel(ExponentIn=ScalarParameter() {defaultValue=2.5, parameterName=ExponentIn_2}) {baseReflectFraction=0}
  n45 = Add(
   A=LinearInterpolate(A=Multiply(A=@n35, B=@n43), B=VectorParameter() {parameterName=Blue, defaultValue=<rgba 0.1119 0.7682 0.7682 1>}.RGB, Alpha=@n44)
   B=NamedRerouteUsage()
  )
  n46 = Multiply(A=NamedRerouteUsage(), B=ScalarParameter() {defaultValue=0.3, parameterName=B_6})
  n47 = AtmosphericLightVector()
  n48 = VertexNormalWS()
  n49 = LinearInterpolate(
   A=ScalarParameter() {defaultValue=0.2, group=SL37 - T1 Butterfly, parameterName=Color Curve SSS Influence}
   B=ScalarParameter() {group=SL37 - T1 Butterfly, parameterName=Color Curve SSS Backlit Influence}
   Alpha=Saturate(None=DotProduct(A=@n47, B=@n48))
  ) {constB=0}
  n50 = SetMaterialAttributes!(
   MaterialAttributes=@n19.MaterialAttributes
   BaseColor=LinearInterpolate(B=VectorParameter() {parameterName=Glitter Color, defaultValue=<rgba 0 1.2728 2 1>}.RGB) {constAlpha=0}
   EmissiveColor=Multiply(A=@n20, B=Add(A=@n45, B=@n46))
   Roughness=LinearInterpolate(A=ScalarParameter() {defaultValue=0.9316, parameterName=Under Roughness}) {constAlpha=0, constB=0.1}
   Opacity=NamedRerouteUsage()
   SubsurfaceColor=Multiply(A=Multiply(A=@n41, B=@n49), B=@n20)
   ShadingModel=ShadingModel() {shadingModel=MSM_Subsurface}
   Specular=LinearInterpolate() {constA=1, constAlpha=0, constB=2}
  ) {attributeSetTypes [
    #1
    #3
    #5
    #8
    #4
    #9
    #6
   ]}
  n51 = Multiply(A=NamedRerouteUsage(), B=ScalarParameter() {defaultValue=1, parameterName=Primary Strength Mask})
  n52 = MaterialFunctionCall!(In=@n50, SkinFX_Mask=@n51) {materialFunction=&/MF_GlitterFX}
  n53 = StaticSwitchParameter(
   True=BlendMaterialAttributes(A=@n18, B=StaticSwitchParameter(True=@n52.Result, False=@n50) {defaultValue=true, group=Glitter, parameterName=Use Glitter?}, Alpha=@n51)
   False=@n18
  ) {defaultValue=true, parameterName=Use Primary?}
  n54 = SetMaterialAttributes!(MaterialAttributes=@n53, BaseColor=LinearInterpolate(Alpha=Fresnel())) {attributeSetTypes [#1]}
  n55 = NamedRerouteUsage()
  n56 = MaterialFunctionCall!(
   Secondary - Curve Amount
   Color Gradient Mask=NamedRerouteUsage()
   Tertiary=NamedRerouteUsage()
   Tertiary Fresnel Color
   Secondary - Curve Index
   "MA (In)"=MaterialFunctionCall!(Secondary=NamedRerouteUsage(), "MA (In)"=BlendMaterialAttributes(A=@n53, B=@n54, Alpha=@n55)) {materialFunction=&/MF_SL_37_SecondaryGlass}."Switch Param (False)"
   Use Clear Coat=StaticBool()
   Use Color Curve
   Distortion Amount
   Metallic=ScalarParameter() {group=SL 37 - Tertiary, parameterName=T1 - Tertiary Metallic}
   Normal Intensity=ScalarParameter() {defaultValue=1, parameterName=T1 - Tertiary Normal Intensity}
  ) {materialFunction=&/MF_SL_37_TertiaryFresnel}
  FunctionOutput(None=MaterialFunctionCall!(
    Material Attributes=MaterialFunctionCall!(Additional Mask=@n7, Outlines=@n8, Outline Color=@n9.RGB, "MA (In)"=@n56."MA (Out)") {materialFunction=&/MF_SL_37_Outlines}.SetMaterialAttributes
    Color=VectorParameter() {parameterName=T1 Decals - Emissive Color, defaultValue=<rgba 4 0.9702 0.0271 1>}.RGB
    Fresnel Exp=ScalarParameter() {defaultValue=0.1353, parameterName=T1 Decal - Fresnel Exp}
    Mask=NamedRerouteUsage()
    Intensity=ScalarParameter() {defaultValue=1, parameterName=T1 Decal - Intensity}
   ) {materialFunction=&/MF_SL_37_Decals}.BlendMaterialAttributes) {outputName=SetMaterialAttributes}
  FunctionOutput(None=@n18) {outputName=OutputPin, sortPriority=10}
  FunctionOutput(None=Max(A=NamedRerouteUsage(), B=NamedRerouteUsage())) {outputName=Combined}
  GetMaterialAttributes() {attributeGetTypes [#3]}
  Multiply(A=NamedRerouteUsage(), B=OneMinus(None=@n33))
  n57 = Normalize(VectorInput=Multiply(A=Add(), B=ScalarParameter() {defaultValue=1, parameterName=Distortion}))
  n58 = Saturate(None=DotProduct(B=OneMinus(None=@n57)))
  n59 = ScalarParameter() {defaultValue=4.0611, parameterName=Power}
  Multiply(A=Multiply(A=Power(Base=@n58, Exp=@n59), B=ScalarParameter() {defaultValue=1, parameterName=Scale}), B=DotProduct(A=CameraVectorWS()))
  Multiply()
  NamedRerouteDeclaration(None=@n19.Specular) {name=Specular In, nodeColor=<rgba 0 1 0.7422 1>}
  NamedRerouteDeclaration(None=@n19.AmbientOcclusion) {name=AO In, nodeColor=<rgba 1 0.4219 0 1>}
  NamedRerouteDeclaration(None=@n19.BaseColor) {name=Base Color In, nodeColor=<rgba 0.1016 0 1 1>}
  n60 = ScalarParameter() {defaultValue=0.4174, parameterName=ExponentIn_3}
  n61 = Multiply(A=OneMinus(None=Fresnel(ExponentIn=@n60) {baseReflectFraction=0}), B=ScalarParameter() {defaultValue=2, parameterName=B_2})
  n62 = OneMinus(None=NamedRerouteUsage())
  n63 = OneMinus(None=Saturate(None=Multiply(A=@n61, B=@n62)))
  n64 = Panner(Coordinate=Multiply(A=TextureCoordinate(), B=ScalarParameter() {defaultValue=1.4696, parameterName=B_5})) {speedY=0.1}
  n65 = ScalarParameter() {defaultValue=-5, parameterName=Height_2}
  n66 = Power(
   Base=TextureSampleParameter2D(UVs=BumpOffset(Coordinate=@n64, Height=@n65)) {parameterName=Param, samplerType=SAMPLERTYPE_Grayscale, texture=&LowResBlurredNoise}.RGB
   Exp=ScalarParameter() {defaultValue=0.5, parameterName=Exp}
  )
  NamedRerouteDeclaration(None=Multiply(A=Multiply(A=@n63, B=@n66), B=ScalarParameter() {defaultValue=2.0841, parameterName=SSS opacit})) {name=Thickness, nodeColor=<rgba 0 1 0.8359 1>}
  n67 = MaterialFunctionCall(Float3=FunctionInput() {inputName="Tech Art Mask (In)", inputType=FunctionInput_Vector4}) {materialFunction=&BreakOutFloat3Components}
  n68 = NamedRerouteDeclaration(None=@n67.R) {name=Decals, nodeColor=<rgba 1 1 0 1>}
  n69 = NamedRerouteDeclaration(None=@n67.G) {name=Outlines, nodeColor=<rgba 0 1 0 1>}
  n70 = NamedRerouteDeclaration(None=@n67.B) {name=Hair, nodeColor=<rgba 0.2188 0 1 1>}
  NamedRerouteDeclaration(None=Max(A=Max(A=@n68, B=@n69), B=@n70)) {name=FX2, nodeColor=<rgba 0 1 0.0156 1>}
  n71 = ScalarParameter() {defaultValue=15, group=SL37 - T1 Butterfly, parameterName=Glitter Tiling Close}
  n72 = ScalarParameter() {defaultValue=50, group=SL37 - T1 Butterfly, parameterName=Glitter Tiling Far}
  n73 = MaterialFunctionCall() {materialFunction=&CameraDepthFade}
  n74 = Panner(
   Coordinate=Multiply(A=TextureCoordinate(), B=LinearInterpolate(A=@n71, B=@n72, Alpha=@n73.Result))
   Speed=MaterialFunctionCall() {materialFunction=&MakeFloat2}.Result
  )
  n75 = TextureSample(UVs=BumpOffset(Coordinate=@n74, Height=ScalarParameter() {defaultValue=-4.5487, parameterName=Glitter Depth})) {texture=/BRCosmetics/Characters/Player/Female/Medium/Bodies/F_MED_HangSpec_Screech/Skins/SuperlevelT1/Textures/T_GlitterNoise_A}
  n76 = TextureSample(UVs=BumpOffset(Coordinate=@n74, Height=ScalarParameter() {parameterName=Height_1})) {texture=&T_GlitterNoise_A}
  n77 = Add(A=@n74) {constB=0.5}
  n78 = ScalarParameter() {defaultValue=-10.659, parameterName=Glitter - Second Layer Depth}
  n79 = Multiply(
   A=TextureSample(UVs=BumpOffset(Coordinate=@n77, Height=@n78)) {texture=&T_GlitterNoise_A}.RGB
   B=ScalarParameter() {defaultValue=0.75, parameterName=Glitter - Second Layer Intensity}
  )
  n80 = Multiply(A=TextureCoordinate(), B=ScalarParameter() {defaultValue=10, group=Paw Jasmine, parameterName=Dots - Modulation Tiling})
  n81 = Add(
   A=Time()
   B=Multiply(
    A=ScalarParameter() {defaultValue=5, group=Paw Jasmine, parameterName="Dots -  Modulation Strength"}
    B=TextureSample(UVs=@n80) {samplerType=SAMPLERTYPE_Grayscale, texture=&LowResBlurredNoise}.R
   )
  )
  n82 = ScalarParameter() {defaultValue=0.5, parameterName=B_9}
  n83 = MaterialFunctionCall!(
   Tex=@n79
   Position=MaterialFunctionCall!(Sine Phase=Multiply(A=@n81, B=@n82), Value 1, Value 2) {materialFunction=/Engine/Functions/Engine_MaterialFunctions02/Utility/Sine_Remapped}.Remapped Sine
   Range
  ) {materialFunction=/BRCosmetics/Characters/Player/Male/Medium/Bodies/M_MED_PawJasmineGravity/Skins/Greet/Materials/MF_HistogramSelect}
  n84 = OneMinus(None=@n79)
  NamedRerouteDeclaration(None=Max(A=Multiply(A=@n75.RGB, B=@n76.RGB), B=Subtract(A=@n83.Result, B=@n84))) {name=Glitter, nodeColor=<rgba 1 0 0.6562 1>}
  n85 = Normalize(VectorInput=VertexNormalWS())
  n86 = DotProduct(A=Normalize(VectorInput=AtmosphericLightVector()), B=@n85)
  n87 = ScalarParameter() {defaultValue=1.0979, parameterName=Front Lit Exp}
  n88 = ScalarParameter() {defaultValue=1, parameterName=Back Lit Exp}
  n89 = CameraVectorWS()
  n90 = OneMinus(None=DotProduct(A=@n89, B=AtmosphericLightVector()))
  n91 = Saturate(None=@n90)
  n92 = Saturate(None=Power(Base=OneMinus(None=@n86), Exp=LinearInterpolate(A=@n87, B=@n88, Alpha=@n91)))
  n93 = Normalize(VectorInput=@n89)
  n94 = ScalarParameter() {defaultValue=14.7913, parameterName=SSS Exponent - Frontlit}
  n95 = ScalarParameter() {defaultValue=2, parameterName=SSS Exponent - Backlit}
  n96 = Saturate(None=Power(Base=DotProduct(A=@n85, B=@n93), Exp=LinearInterpolate(A=@n94, B=@n95, Alpha=@n91)))
  n97 = ScalarParameter() {defaultValue=5, parameterName=Frontlit}
  n98 = ScalarParameter() {parameterName=Backlit}
  n99 = Multiply(A=@n62, B=Multiply(A=Multiply(A=@n92, B=@n96), B=LinearInterpolate(A=@n97, B=@n98, Alpha=@n90)))
  n100 = NamedRerouteUsage()
  NamedRerouteDeclaration(None=Saturate(None=Multiply(A=@n99, B=@n100))) {name=Backlit, nodeColor=<rgba 0 0.25 1 1>}
  n101 = PixelNormalWS()
  n102 = TextureSample(UVs=ReflectionVectorWS(CustomWorldNormal=@n101)) {samplerType=SAMPLERTYPE_LinearColor, texture=/BRCosmetics/Accessories/FORT_Backpacks/Backpack_F_MED_Fuzzy_Bear_Summer/Texture/AthenaCubemap2}
  NamedRerouteDeclaration(None=@n102.RGB) {name=Surface Reflection, nodeColor=<rgba 0.0078 0 1 1>}
  NamedRerouteDeclaration(None=Max(A=Max(A=@n4, B=@n5), B=@n6)) {name=FX, nodeColor=<rgba 1 0 0.3047 1>}
  NamedRerouteDeclaration(None=Step(X=@n70) {constY=0.05}) {name=HairStepped, nodeColor=<rgba 1 0 0.2109 1>}
  n103 = ComponentMask(None=VertexInterpolator(VS=LocalPosition() {localOrigin=InstancePreSkinning}.XYZ).PS) {b=true}
  n104 = ScalarParameter() {defaultValue=151.0474, parameterName=Preskinned Mask - Length}
  NamedRerouteDeclaration(None=Saturate(None=Divide(A=@n103, B=@n104))) {name=Preskinned Mask, nodeColor=<rgba 0.1953 0 1 1>}
  NamedRerouteDeclaration(None=@n19.Roughness) {name=Roughness In, nodeColor=<rgba 0 0.1094 1 1>}
  NamedRerouteUsage()
  OneMinus(None=@n44)
  OneMinus(None=Fresnel(ExponentIn=ScalarParameter() {defaultValue=0.1233, parameterName=ExponentIn}) {baseReflectFraction=0})
  ScalarParameter() {defaultValue=0.5, parameterName="Butterfly  - Emissive Strength"}
  ScalarParameter() {defaultValue=-1.5489, parameterName=Top Layer Roughness}
  n105 = TextureCoordinate()
  n106 = ScalarParameter() {defaultValue=80, parameterName=Tiling Normal}
  TextureSample(UVs=MaterialFunctionCall(UVs=Multiply(A=@n105, B=@n106), "Rotation Angle (0-1)"=ScalarParameter() {defaultValue=0.5, parameterName="Rotation Angle (0-1) (S)"}) {materialFunction=&CustomRotator}.Rotated Values) {
   samplerType=SAMPLERTYPE_Normal
   texture=/BRCosmetics/Characters/Player/Female/Medium/Bodies/F_MED_HangSpec_Screech/Skins/SuperlevelT1/Textures/T_Butterfly_N_01
  }
  n107 = TextureCoordinate()
  TextureSample(UVs=Panner(
    Coordinate=Multiply(A=@n107, B=@n27)
    Speed=MaterialFunctionCall(X=@n28.R, Y=@n28.R) {materialFunction=&MakeFloat2}.Result
   )) {samplerType=SAMPLERTYPE_Grayscale, texture=&LowResBlurredNoise}
  n108 = Multiply(
   A=TextureCoordinate()
   B=MaterialFunctionCall(
    X=ScalarParameter() {defaultValue=2.5, parameterName=Butterfly Pattern - X -Tiling}
    Y=ScalarParameter() {defaultValue=1.3236, parameterName=Butterfly Pattern - Y -Tiling}
   ) {materialFunction=&MakeFloat2}.Result
  )
  TextureSample(UVs=BumpOffset(Coordinate=Add(A=@n108, B=@n31), Height=@n32)) {
   samplerType=SAMPLERTYPE_Masks
   texture=&T_Butterfly_M_01
  }
  Transform(None=@n101)
  VectorParameter() {parameterName=Glitter Speed, defaultValue=<rgba 0 0.0655 0 0>}
  VertexNormalWS()
 ]}
