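# Parent material function for a Season 37 super-level character skin, restyling incoming attributes through static-switched Butterfly, Spider and Beetle tiers.
# Skin gets subsurface profile shading, then the Butterfly function runs; its mask and the Tech Art mask drive region blends.
# Spider and Beetle darken paint and add clear coat with, respectively, curve-atlas fresnel plus fireflies, or warped procedural waves.
# Both share glass, tertiary fresnel, decal and outline subfunctions, using pre-skinned position for animated patterns.
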MaterialFunction /BRCosmetics/Characters/Player/Female/Medium/Bodies/F_MED_HangSpec_Screech/Skins/MF_S37_SL_Parent
graph {nodes [
  n1 = TextureCoordinate()
  n2 = Constant() {r=15}
  n3 = Multiply(A=ComponentMask(None=@n1) {r=true}, B=@n2) {constB=5}
  Add(A=Time(), B=@n3)
  Constant() {r=1}
  Constant() {r=1}
  Constant() {r=1}
  n4 = VectorParameter() {group=01_Eye_Advanced, parameterName=Eye Camera Light Vector, sortPriority=36, defaultValue=<rgba -0.2 0.075 -0.5 0>}
  DotProduct(A=Normalize(VectorInput=VertexNormalWS()), B=Transform(None=Normalize(VectorInput=@n4.RGB)) {transformSourceType=TRANSFORMSOURCE_View})
  n5 = NamedRerouteUsage()
  n6 = NamedRerouteUsage()
  n7 = VectorParameter() {parameterName=T3 - Outline Color, defaultValue=<rgba 1.5 0.0058 0.3247 1>}
  n8 = ScalarParameter() {
   defaultValue=13
   desc="Color Curve \r\n\r\nThis is the total amount of color curves in your Curve Atlas"
   group=SL37 - T2 Spider
   parameterName=T3 Secondary - Curve Amount
  }
  n9 = NamedRerouteUsage()
  n10 = NamedRerouteUsage()
  n11 = VectorParameter() {group=SL37 - T2 Spider, parameterName=T3 - Teritary Fresnel Color, defaultValue=<rgba 0.4287 0.956 0.0409 1>}
  n12 = ScalarParameter() {
   defaultValue=12
   desc="Metallic - Color Curve \r\n\r\nThis number allows you to choose which index you want to use in your Curve Atlas so you don't need to use multiple curve atlas for different color curves"
   group=SL37 - T2 Spider
   parameterName=T3 Secondary - Curve Index
  }
  n13 = NamedRerouteUsage()
  n14 = NamedRerouteUsage()
  n15 = ScalarParameter() {
   defaultValue=13
   desc="Color Curve \r\n\r\nThis is the total amount of color curves in your Curve Atlas"
   group=SL37 - T2 Spider
   parameterName=T2 Secondary - Curve Amount
  }
  n16 = NamedRerouteUsage()
  n17 = NamedRerouteUsage()
  n18 = VectorParameter() {group=SL37 - T2 Spider, parameterName=T2 - Teritary Fresnel Color, defaultValue=<rgba 0.7682 0 0.0065 1>}
  n19 = ScalarParameter() {
   defaultValue=11
   desc="Metallic - Color Curve \r\n\r\nThis number allows you to choose which index you want to use in your Curve Atlas so you don't need to use multiple curve atlas for different color curves"
   group=SL37 - T2 Spider
   parameterName=T2 Secondary - Curve Index
  }
  n20 = Subtract(A=OneMinus(None=NamedRerouteUsage()), B=NamedRerouteUsage())
  n21 = FunctionInput() {inputName=Result, inputType=FunctionInput_MaterialAttributes}
  n22 = NamedRerouteDeclaration(None=FunctionInput() {bUsePreviewValueAsDefault=true, inputName=Skin Mask, inputType=FunctionInput_Scalar}) {name=SkinMask, nodeColor=<rgba 1 0 0.3047 1>}
  n23 = BlendMaterialAttributes(
   A=@n21
   B=SetMaterialAttributes!(MaterialAttributes=GetMaterialAttributes(None=@n21).MaterialAttributes, ShadingModel=ShadingModel() {shadingModel=MSM_SubsurfaceProfile}) {attributeSetTypes [#1]}
   Alpha=@n22
  )
  n24 = StaticSwitchParameter(
   True=SetMaterialAttributes!(
    MaterialAttributes=@n23
    BaseColor=TextureSample() {texture=/BRCosmetics/Characters/Player/Female/Medium/Bodies/F_MED_SwoopClasp_Pry/Textures/T_SwoopClasp_Pry_Body_SL_FX2}.RGB
   ) {attributeSetTypes [#2]}
   False=@n23
  ) {group=SL 37 - Global, parameterName=Debug Breakout Masks}
  n25 = FunctionInput() {inputName=Tech Art Mask, inputType=FunctionInput_Vector4, sortPriority=10}
  n26 = MaterialFunctionCall!(
   "MA (In)"=@n24
   "Tech Art Mask (In)"=@n25
   Skin Mask=@n22
   M=FunctionInput() {bUsePreviewValueAsDefault=true, inputName=M, inputType=FunctionInput_Scalar}
  ) {materialFunction=&/MF_Butterfly_T1}
  n27 = StaticSwitchParameter(True=@n26.SetMaterialAttributes, False=@n24) {defaultValue=true, group=SL 37 - Global, parameterName=Tier 1 - Butterfly}
  n28 = GetMaterialAttributes(None=@n27) {attributeGetTypes [#2, #3]}
  n29 = Power(Base=Desaturation(None=@n28.BaseColor), Exp=ScalarParameter() {defaultValue=0.1, group=SL37 - T2 Spider, parameterName=T2 - Desat Pow})
  n30 = LinearInterpolate(
   A=ScalarParameter() {defaultValue=0.015, group=SL37 - T2 Spider, parameterName=T2 - Paint Strength}
   B=ScalarParameter() {defaultValue=0.1, group=SL37 - T2 Spider, parameterName=T2 - Metal Paint Strength}
   Alpha=GetMaterialAttributes(None=@n24) {attributeGetTypes [#3]}.Metallic
  )
  n31 = ScalarParameter() {defaultValue=0.6, parameterName=T2 - Metallic Multipler}
  n32 = BlendMaterialAttributes(
   A=@n27
   B=SetMaterialAttributes!(MaterialAttributes=@n28.MaterialAttributes, BaseColor=Multiply(A=@n29, B=@n30), Metallic=Multiply(A=@n28.Metallic, B=@n31)) {attributeSetTypes [#2, #3]}
   Alpha=@n20
  )
  n33 = Constant()
  n34 = BlendMaterialAttributes(
   A=@n32
   B=SetMaterialAttributes!(MaterialAttributes=GetMaterialAttributes(None=@n32).MaterialAttributes, EmissiveColor=@n33, BaseColor, SubsurfaceColor=@n33) {attributeSetTypes [#4, #2, #5]}
   Alpha=@n20
  )
  n35 = GetMaterialAttributes(None=@n34) {attributeGetTypes [#2, #6]}
  n36 = Multiply(A=Fresnel(ExponentIn=ScalarParameter() {defaultValue=1.64, group=SL37 - T2 Spider, parameterName=T2 - Spider Fresnel Exp}) {baseReflectFraction=0}) {constB=2}
  n37 = Frac(None=@n36)
  n38 = VectorParameter() {group=SL37 - T2 Spider, parameterName=T2 - Spider Fresnel Color, defaultValue=<rgba 0.1 0.0282 0.0904 1>}
  n39 = Fresnel(ExponentIn=ScalarParameter() {defaultValue=1.2953, group=SL37 - T2 Spider, parameterName=T2 Fresnel}) {baseReflectFraction=0}
  n40 = Multiply(A=TextureCoordinate(), B=ScalarParameter() {defaultValue=4.2896, group=SL37 - T2 Spider, parameterName=Fresnel Noise Tiling})
  n41 = ComponentMask(None=VectorParameter() {group=SL37 - T2 Spider, parameterName=Fresnel Noise Speed, defaultValue=<rgba 0 0.5 0 0>}.RGB) {g=true, r=true}
  n42 = Power(
   Base=TextureSample(UVs=Panner(Coordinate=@n40, Speed=@n41) {speedY=0.1}) {samplerType=SAMPLERTYPE_Grayscale, texture=&LowResBlurredNoise}.RGB
   Exp=ScalarParameter() {defaultValue=1.2, group=SL37 - T2 Spider, parameterName=Fresnel Noise Power}
  )
  n43 = Multiply(
   A=NamedRerouteUsage()
   B=Multiply(
    A=MaterialFunctionCall!(Base=@n39, Add=@n42) {materialFunction=/Engine/Functions/Engine_MaterialFunctions02/Math/MultiplyAdd}.Result
    B=VectorParameter() {group=SL37 - T2 Spider, parameterName=T2 Fresnel Color, defaultValue=<rgba 10 0.0342 0.2083 1>}.RGB
   )
  )
  n44 = ScalarParameter() {
   desc="Metallic - Color Curve \r\n\r\nThis number allows you to choose which index you want to use in your Curve Atlas so you don't need to use multiple curve atlas for different color curves"
   group=SL37 - T2 Spider
   parameterName=Curve Index
   sortPriority=30
  }
  n45 = ScalarParameter() {
   defaultValue=13
   desc="Color Curve \r\n\r\nThis is the total amount of color curves in your Curve Atlas"
   group=SL37 - T2 Spider
   parameterName=Curve Amount
   sortPriority=29
  }
  n46 = NamedRerouteUsage()
  n47 = TextureObjectParameter() {
   group=SL 37 - Global
   parameterName=Color Curves
   samplerType=SAMPLERTYPE_LinearColor
   texture=/BRCosmetics/Characters/Player/Female/Medium/Bodies/F_MED_HangSpec_Screech/Skins/SuperlevelT1/Textures/CA_SL37
  }
  n48 = TextureSample(
   UVs=Saturate(None=MaterialFunctionCall!(CurveIndex=@n44, CurveNum=@n45, GreyMapIn=@n46) {materialFunction=/Game/Effects/Niagara/MasterMaterials/Functions/Common/FXCommon_CurveAtlasUV}.UV)
   Tex=@n47
  ) {
   samplerSource=SSM_Wrap_WorldGroupSettings
   samplerType=SAMPLERTYPE_LinearColor
   texture=&CA_SL37
  }
  n49 = ScalarParameter() {defaultValue=0.2448, group=SL37 - T2 Spider, parameterName=Color Curve - Step}
  n50 = ScalarParameter() {defaultValue=0.3, group=SL37 - T2 Spider, parameterName=Color Curve - Step Smoothness}
  n51 = Fresnel(ExponentIn=ScalarParameter() {defaultValue=0.5266, group=SL37 - T2 Spider, parameterName=Color Curve - Fresnel Exp})
  n52 = Multiply(A=@n48.RGB, B=SmoothStep(Min=@n49, Max=Add(A=@n49, B=@n50), Value=OneMinus(None=@n51)))
  n53 = ScalarParameter() {defaultValue=0.15, group=SL37 - T2 Spider, parameterName=Main Firefly - Radius}
  n54 = ScalarParameter() {defaultValue=-0.018, group=SL37 - T2 Spider, parameterName=Main Firefly - Hardness}
  n55 = VectorParameter() {group=SL37 - T2 Spider, parameterName=Main Firefly - Speed, defaultValue=<rgba 0 0 -4 1>}
  n56 = ScalarParameter() {defaultValue=1, group=SL37 - T2 Spider, parameterName=Main Firefly - Depth}
  n57 = ScalarParameter() {defaultValue=0.4, group=SL37 - T2 Spider, parameterName=Main Firefly - Tile}
  n58 = ScalarParameter() {defaultValue=0.1, group=SL37 - T2 Spider, parameterName=Step Fireflies}
  n59 = Add(A=@n58, B=ScalarParameter() {defaultValue=-0.0488, group=SL37 - T2 Spider, parameterName=Step Fireflies Smoothness})
  n60 = ComponentMask(None=VertexInterpolator(VS=LocalPosition() {localOrigin=InstancePreSkinning}.XYZ).PS) {b=true}
  n61 = ScalarParameter() {defaultValue=1251.5172, group=SL37 - T2 Spider, parameterName=Firefly Mask - Length}
  n62 = SmoothStep(Min=@n58, Max=@n59, Value=Saturate(None=Divide(A=@n60, B=@n61))) {constMin=0.1}
  n63 = Divide(A=@n60, B=ScalarParameter() {defaultValue=1869.2572, group=SL37 - T2 Spider, parameterName=Firefly Mask 2 - Length})
  n64 = OneMinus(None=SmoothStep(Min=@n58, Max=@n59, Value=Saturate(None=@n63)) {constMin=0.1})
  n65 = ScalarParameter() {defaultValue=0.1, parameterName=Step Fireflies}
  n66 = ScalarParameter() {defaultValue=-0.0488, parameterName=Step Fireflies Smoothness}
  n67 = ComponentMask(None=VertexInterpolator(VS=LocalPosition() {localOrigin=InstancePreSkinning}.XYZ).PS) {b=true}
  n68 = Divide(A=@n67, B=ScalarParameter() {defaultValue=700, group=SL37 - T2 Spider, parameterName=Firefly Mask 3 - Length})
  n69 = Saturate(None=Divide(A=@n67, B=ScalarParameter() {defaultValue=100, group=SL37 - T2 Spider, parameterName=Firefly Mask 4 - Length}))
  n70 = Min(A=SmoothStep(Min=@n65, Max=Add(A=@n65, B=@n66), Value=Saturate(None=@n68)) {constMin=0.1}, B=OneMinus(None=Step(Y=@n69) {constX=0.1}))
  n71 = ScalarParameter() {defaultValue=0.1, group=SL37 - T2 Spider, parameterName=Firefly 02 - Intensity}
  n72 = ScalarParameter() {defaultValue=0.1, parameterName=Step Fireflies}
  n73 = ScalarParameter() {defaultValue=-0.0488, parameterName=Step Fireflies Smoothness}
  n74 = ComponentMask(None=VertexInterpolator(VS=LocalPosition() {localOrigin=InstancePreSkinning}.XYZ).PS) {b=true}
  n75 = Divide(A=@n74, B=ScalarParameter() {defaultValue=270.1391, group=SL37 - T2 Spider, parameterName=Firefly Mask 5 - Length})
  n76 = Saturate(None=Divide(A=@n74, B=ScalarParameter() {defaultValue=63.117, group=SL37 - T2 Spider, parameterName=Firefly Mask 6 - Length}))
  n77 = Min(A=SmoothStep(Min=@n72, Max=Add(A=@n72, B=@n73), Value=Saturate(None=@n75)) {constMin=0.1}, B=OneMinus(None=Step(Y=@n76) {constX=0.1}))
  n78 = ScalarParameter() {group=SL37 - T2 Spider, parameterName=Firefly 03 - Intensity}
  n79 = Max(
   A=Multiply(A=Min(A=@n62, B=@n64), B=ScalarParameter() {defaultValue=1, group=SL37 - T2 Spider, parameterName=Firefly 01 - Intensity})
   B=Max(A=Multiply(A=@n70, B=@n71), B=Multiply(A=@n77, B=@n78))
  )
  n80 = Multiply(
   A=Multiply(
    A=MaterialFunctionCall!(Radius=@n53, Hardness=@n54, Speed=@n55.RGB, Depth=@n56, Offset, Random Speed, Tile=@n57) {materialFunction=&/MF_DotsRandom_WP}.Dots
    B=ScalarParameter() {defaultValue=40, group=SL37 - T2 Spider, parameterName=Main Firefly - Intensity, sortPriority=6}
   ) {constB=20}
   B=@n79
  )
  n81 = Multiply(
   A=MaterialFunctionCall(
    Radius=ScalarParameter() {defaultValue=0.1, group=SL37 - T2 Spider, parameterName=Secondary Firefly - Radius}
    Hardness=ScalarParameter() {defaultValue=0.35, group=SL37 - T2 Spider, parameterName=Secondary Firefly - Hardness}
    Speed=VectorParameter() {group=SL37 - T2 Spider, parameterName=Secondary Firefly - Speed, defaultValue=<rgba 0 0 -4 1>}.RGB
    Depth=ScalarParameter() {defaultValue=3, group=SL37 - T2 Spider, parameterName=Secondary Firefly - Depth}
    Offset=VectorParameter() {group=SL37 - T2 Spider, parameterName=Secondary Firefly - Offset}.RGB
    Tile=ScalarParameter() {defaultValue=0.4, group=SL37 - T2 Spider, parameterName=Secondary Firefly - Tile}
   ) {materialFunction=&MF_DotsRandom_WP}.Dots
   B=ScalarParameter() {defaultValue=0.2, group=SL37 - T2 Spider, parameterName=Secondary Firefly - Intensity}
  )
  n82 = MaterialFunctionCall(
   Radius=ScalarParameter() {defaultValue=0.125, group=SL37 - T2 Spider, parameterName=Third Firefly - Radius}
   Hardness=ScalarParameter() {defaultValue=0.35, group=SL37 - T2 Spider, parameterName=Third Firefly - Hardness}
   Speed=VectorParameter() {group=SL37 - T2 Spider, parameterName=Third Firefly - Speed, defaultValue=<rgba 0 0 -2 1>}.RGB
   Depth=ScalarParameter() {defaultValue=2, group=SL37 - T2 Spider, parameterName=Third Firefly - Depth}
   Offset=VectorParameter() {group=SL37 - T2 Spider, parameterName=Third Firefly - Offset, defaultValue=<rgba 45 0 0 0>}.RGB
   Tile=ScalarParameter() {defaultValue=0.4, group=SL37 - T2 Spider, parameterName=Third Firefly - Tile}
  ) {materialFunction=&MF_DotsRandom_WP}
  n83 = ScalarParameter() {defaultValue=0.4, group=SL37 - T2 Spider, parameterName=Third Firefly - Intensity}
  n84 = Multiply(
   A=Max(A=Max(A=@n80, B=@n81), B=Multiply(A=@n82.Dots, B=@n83))
   B=ScalarParameter() {defaultValue=1.3, parameterName=Firefly - Overall Intensity}
  )
  n85 = OneMinus(None=Fresnel(ExponentIn=ScalarParameter() {defaultValue=0.6294, group=SL37 - T2 Spider, parameterName=Firefly - Fresnel Cutoff}) {baseReflectFraction=0})
  n86 = Max(
   A=Max(A=Multiply(A=@n37, B=@n38.RGB), B=Max(A=@n43, B=@n52))
   B=Multiply(
    A=VectorParameter() {group=SL37 - T2 Spider, parameterName=Firefly Color, defaultValue=<rgba 0.9823 0.0802 0.0003 1>}.RGB
    B=Multiply(A=@n84, B=@n85)
   )
  )
  n87 = ScalarParameter() {defaultValue=0.824, parameterName=T2 - Shine Step}
  n88 = ScalarParameter() {defaultValue=0.208, parameterName=T2 - Shine Smoothness}
  n89 = Fresnel(ExponentIn=ScalarParameter() {defaultValue=1, group=SL37 - T2 Spider, parameterName=T2 - Shine Fresnel})
  n90 = Multiply(
   A=SmoothStep(Min=@n87, Max=Add(A=@n87, B=@n88), Value=OneMinus(None=@n89))
   B=ScalarParameter() {defaultValue=0.2, group=SL37 - T2 Spider, parameterName=T2 - Shine Amount}
  )
  n91 = TextureSample(UVs=Multiply(A=TextureCoordinate(), B=ScalarParameter() {defaultValue=20, group=SL37 - T2 Spider, parameterName=T2 - Detail Normal - Tiling})) {samplerType=SAMPLERTYPE_Normal, texture=/Game/Global/Textures/Default/Tileable_Patterns/T_Skyfire_DarkGlass_DetailNormal}
  n92 = ScalarParameter() {defaultValue=0.7, group=SL37 - T2 Spider, parameterName=T2 - Detail Normal - Strength}
  n93 = SetMaterialAttributes!(
   MaterialAttributes=@n35.MaterialAttributes
   BaseColor=Constant()
   ShadingModel=ShadingModel() {shadingModel=MSM_ClearCoat}
   CustomData0=Constant() {r=1}
   CustomData1=ScalarParameter() {defaultValue=0.2, parameterName="Clear  Coat - Roughness"}
   EmissiveColor=NamedRerouteDeclaration(None=Add(A=@n86, B=@n90)) {name=T2 - Primary Emissive, nodeColor=<rgba 1 0 0.8203 1>}
   Normal=MaterialFunctionCall(
    BaseNormal=@n35.Normal
    AdditionalNormal=MaterialFunctionCall(Normal=@n91.RGB, Flatness=@n92) {materialFunction=&FlattenNormal}.Result
   ) {materialFunction=&BlendAngleCorrectedNormals}.Result
   Metallic=Constant()
  ) {attributeSetTypes [
    #2
    #1
    #7
    #8
    #4
    #6
    #3
   ]}
  n94 = MaterialFunctionCall(Float4=@n26.SL Mask) {materialFunction=&BreakOutFloat4Components}
  n95 = NamedRerouteDeclaration(None=@n94.R) {name=Primary, nodeColor=<rgba 1 0 0 1>}
  n96 = MaterialFunctionCall!(
   Additional Mask=NamedRerouteUsage()
   Outlines=NamedRerouteUsage()
   Outline Color=VectorParameter() {group=SL37 - T2 Spider, parameterName=T2 - Outline Emissive Color, defaultValue=<rgba 1.7 0 0.0191 1>}.RGB
   "MA (In)"=MaterialFunctionCall!(Secondary=NamedRerouteUsage(), "MA (In)"=BlendMaterialAttributes(A=@n34, B=@n93, Alpha=@n95)) {materialFunction=&/MF_SL_37_SecondaryGlass}."Switch Param (False)"
  ) {materialFunction=&/MF_SL_37_Outlines}
  n97 = StaticBoolParameter() {group=SL 37 - Tertiary, parameterName=T2 - Use Clear Coat}
  n98 = ScalarParameter() {group=SL 37 - Tertiary, parameterName=T2 - Tertiary Metallic}
  n99 = ScalarParameter() {defaultValue=1, group=SL37 - T2 Spider, parameterName=T2 - Tertiary Normal Intensity}
  n100 = GetMaterialAttributes(None=StaticSwitchParameter(
    True=MaterialFunctionCall!(
     Secondary - Curve Amount=@n15
     Color Gradient Mask=@n16
     Tertiary=@n17
     Tertiary Fresnel Color=@n18.RGB
     Secondary - Curve Index=@n19
     "MA (In)"=@n96.SetMaterialAttributes
     Use Clear Coat=@n97
     Use Color Curve
     Distortion Amount
     Metallic=@n98
     Normal Intensity=@n99
    ) {materialFunction=&/MF_SL_37_TertiaryFresnel}."MA (Out)"
    False=@n96.SetMaterialAttributes
   ) {defaultValue=true, parameterName=Use Tertiary}) {attributeGetTypes [#2]}
  n101 = Max(A=NamedRerouteUsage(), B=NamedRerouteUsage())
  n102 = LinearInterpolate(
   A=@n100.BaseColor
   B=Multiply(A=Desaturation(None=@n100.BaseColor), B=ScalarParameter() {defaultValue=0.1, group=SL37 - T2 Spider, parameterName=T2 - Paint Strength})
   Alpha=OneMinus(None=Round(None=@n101))
  )
  n103 = VectorParameter() {group=SL37 - T2 Spider, parameterName=T2 Decals - Emissive Color, defaultValue=<rgba 3 0.3284 0.2111 1>}
  n104 = StaticSwitchParameter(
   True=MaterialFunctionCall!(
    Material Attributes=SetMaterialAttributes!(MaterialAttributes=@n100.MaterialAttributes, BaseColor=@n102) {attributeSetTypes [#2]}
    Color=@n103.RGB
    Fresnel Exp=ScalarParameter() {defaultValue=0.0644, group=SL37 - T2 Spider, parameterName=T2 Decal - Fresnel Exp}
    Mask=NamedRerouteUsage()
    Intensity=ScalarParameter() {defaultValue=1, group=SL37 - T2 Spider, parameterName=T2 Decal - Intensity}
   ) {materialFunction=&/MF_SL_37_Decals}.BlendMaterialAttributes
   False=@n27
  ) {group=SL 37 - Global, parameterName=Tier 2 - Spider}
  n105 = Constant()
  n106 = SetMaterialAttributes!(
   MaterialAttributes=GetMaterialAttributes(None=@n104).MaterialAttributes
   EmissiveColor=@n105
   BaseColor=@n105
   SubsurfaceColor=@n105
   Roughness
   Specular
  ) {attributeSetTypes [
    #4
    #2
    #5
    #9
    #10
   ]}
  n107 = GetMaterialAttributes(None=@n104) {attributeGetTypes [#2, #3]}
  n108 = Desaturation(None=@n107.BaseColor)
  n109 = ScalarParameter() {defaultValue=0.1, parameterName=T3 - Desat Power}
  n110 = Constant()
  n111 = SetMaterialAttributes!(
   MaterialAttributes=@n107.MaterialAttributes
   BaseColor=Multiply(A=Power(Base=@n108, Exp=@n109), B=ScalarParameter() {defaultValue=0.02, parameterName=T3 - Paint Intensity}) {constB=0.1}
   EmissiveColor=@n110
   SubsurfaceColor=@n110
   Metallic=Multiply(A=@n107.Metallic, B=ScalarParameter() {defaultValue=1, parameterName=T3 - Metallic})
  ) {attributeSetTypes [
    #2
    #4
    #5
    #3
   ]}
  n112 = OneMinus(None=NamedRerouteUsage())
  n113 = NamedRerouteUsage()
  n114 = BlendMaterialAttributes(
   A=BlendMaterialAttributes(A=@n104, B=@n111, Alpha=Subtract(A=@n112, B=@n113))
   B=@n106
   Alpha=Subtract(A=Add(A=@n13, B=@n14), B=NamedRerouteUsage())
  )
  n115 = VectorParameter() {parameterName=T3 - Black Lines, defaultValue=<rgba 0.0024 0.0024 0.0024 1>}
  n116 = ScalarParameter() {defaultValue=-0.22, parameterName=T3 - Inner Coat Step}
  n117 = Add(A=@n116, B=ScalarParameter() {defaultValue=1, parameterName=T3 - Inner Coat Smoothness})
  n118 = OneMinus(None=Fresnel(ExponentIn=ScalarParameter() {defaultValue=1, parameterName=T3 - Inner Coat Fresnel}) {baseReflectFraction=0})
  n119 = ScalarParameter() {defaultValue=0.6, parameterName=T3 - Inner Coat Fresnel Shine Step}
  n120 = Add(A=@n119, B=ScalarParameter() {defaultValue=0.168, parameterName=T3 - Inner Coat Fresnel Shine Smoothness})
  n121 = OneMinus(None=Fresnel(ExponentIn=ScalarParameter() {defaultValue=0.5, parameterName=T3 - Inner Coat Fresnel Shine}) {baseReflectFraction=0})
  n122 = Add(
   A=Multiply(
    A=VectorParameter() {parameterName=T3 - Grey Lines, defaultValue=<rgba 0.0146 0.0146 0.0156 1>}.RGB
    B=SmoothStep(Min=@n116, Max=@n117, Value=@n118)
   )
   B=Multiply(A=SmoothStep(Min=@n119, Max=@n120, Value=@n121), B=ScalarParameter() {defaultValue=0.04, parameterName=T3 - Shine Intensity})
  )
  n123 = ScalarParameter() {defaultValue=-0.4, parameterName=Sine Step}
  n124 = Divide(A=VertexInterpolator(VS=LocalPosition() {localOrigin=InstancePreSkinning}.XYZ).PS) {constB=150}
  n125 = ComponentMask(None=@n124) {r=true}
  n126 = ComponentMask(None=@n124) {b=true}
  n127 = ScalarParameter() {defaultValue=50, parameterName=TilingTest}
  n128 = Time()
  n129 = ScalarParameter() {defaultValue=-0.3, parameterName=Up Movement}
  n130 = AppendVector(A=@n125, B=Add(A=Multiply(A=@n126, B=@n127) {constB=5}, B=Multiply(A=@n128, B=@n129) {constB=0.5}))
  n131 = LocalPosition() {localOrigin=InstancePreSkinning}
  n132 = Multiply(
   A=ComponentMask(None=VertexInterpolator(VS=@n131.XYZ).PS) {g=true, r=true}
   B=ScalarParameter() {defaultValue=0.0075, parameterName=World Scaling}
  ) {constB=0.1}
  n133 = Multiply(A=@n132, B=ScalarParameter() {defaultValue=15, parameterName=Noise Distortion Tiling})
  n134 = Multiply(A=Time()) {constB=0.1}
  n135 = Multiply(
   A=TextureSample(UVs=Add(A=@n133, B=@n134)) {samplerType=SAMPLERTYPE_Grayscale, texture=&LowResBlurredNoise}.R
   B=ScalarParameter() {defaultValue=1, parameterName=Noise Distortion Intensity}
  )
  n136 = Add(A=@n130, B=@n135)
  n137 = ComponentMask(None=@n136) {r=true}
  n138 = Frac(None=Add(A=ComponentMask(None=@n136) {g=true}) {constB=0.2})
  n139 = Time()
  n140 = Add(
   A=Multiply(A=@n125, B=ScalarParameter() {defaultValue=30, parameterName=Warp}) {constB=5}
   B=Multiply(A=@n139, B=ScalarParameter() {defaultValue=0.1, parameterName=Warp 1 Speed})
  )
  n141 = Multiply(A=@n139, B=ScalarParameter() {defaultValue=0.01, parameterName=Warp 2 Speed})
  n142 = Constant() {r=-1}
  n143 = MaterialFunctionCall!(
   Sine Phase=Multiply(A=Add(A=@n125, B=@n141) {constB=0.5}, B=ScalarParameter() {defaultValue=20, parameterName=Warp 2}) {constB=20}
   Value 1=@n142
   Value 2=Constant() {r=1}
  ) {materialFunction=/Engine/Functions/Engine_MaterialFunctions02/Utility/Sine_Remapped}
  n144 = Multiply(A=@n139, B=ScalarParameter() {defaultValue=0.03, parameterName=Warp 3 Speed})
  n145 = MaterialFunctionCall(
   Sine Phase=Multiply(A=Add(A=@n125, B=@n144) {constB=0.5}, B=ScalarParameter() {defaultValue=10, parameterName=Warp 3}) {constB=20}
   Value 1=@n142
   Value 2=ScalarParameter() {defaultValue=1, parameterName=Warp 3 - Max}
  ) {materialFunction=&Sine_Remapped}
  n146 = Subtract(A=Multiply(
    A=MaterialFunctionCall(Sine Phase=@n140) {materialFunction=&Sine_Remapped}.Remapped Sine
    B=Multiply(A=@n143.Remapped Sine, B=@n145.Remapped Sine)
   )) {constB=-1.5}
  n147 = Constant() {r=1}
  n148 = SmoothStep(
   Min=Add(A=@n123, B=ScalarParameter() {defaultValue=0.1368, parameterName=Sine Smoothness})
   Max=@n123
   Value=MaterialFunctionCall!(
    UV=AppendVector(A=@n137, B=@n138)
    Wave Height=ScalarParameter() {defaultValue=0.8}
    Wave Length
    Wave Speed=ScalarParameter() {defaultValue=0.5, parameterName="Wave Speed (S)"}
    GlowMin
    GlowMax
    Outline Thickness=ScalarParameter() {defaultValue=0.7, parameterName="Outline Thickness (S)"}
    Custom Curve=AppendVector(A=@n146, B=@n147)
   ) {materialFunction=/BRCosmetics/Characters/Player/Female/Medium/Bodies/F_MED_HangSpec_Screech/Skins/Material_Functions/MF_Wave}.SDF Outline
  )
  n149 = Fresnel(ExponentIn=ScalarParameter() {defaultValue=3, parameterName=Waves Fresnel}) {baseReflectFraction=0}
  n150 = ScalarParameter() {defaultValue=0.1, parameterName=Waves Fresnel Step}
  n151 = Saturate(None=Multiply(A=OneMinus(None=@n148), B=Step(Y=@n149, X=@n150)))
  n152 = ScalarParameter() {defaultValue=0.01, parameterName=T3 - Fresnel Step}
  n153 = ScalarParameter() {defaultValue=0.5, parameterName=T3 - Fresnel Step Smoothness}
  n154 = ScalarParameter() {defaultValue=4, parameterName=T3 - Fresnel Amount}
  n155 = Fresnel(ExponentIn=ScalarParameter() {defaultValue=1, parameterName=T3 - Fresnel Noise Amount}) {baseReflectFraction=0}
  n156 = LocalPosition()
  n157 = ComponentMask(None=@n156.XYZ) {r=true}
  n158 = TextureSample(UVs=Multiply(A=AppendVector(A=@n157, B=@n156.Z), B=ScalarParameter() {defaultValue=0.005, parameterName=T3 - Fresnel Noise Tiling})) {samplerType=SAMPLERTYPE_Grayscale, texture=&LowResBlurredNoise}
  n159 = Max(
   A=SmoothStep(Min=@n152, Max=Add(A=@n152, B=@n153), Value=Fresnel(ExponentIn=@n154) {baseReflectFraction=0})
   B=Multiply(A=Multiply(A=@n155, B=@n158.R), B=ScalarParameter() {defaultValue=1, parameterName=T3 - Fresnel Noise})
  )
  n160 = LocalPosition() {localOrigin=InstancePreSkinning}
  n161 = Divide(A=OneMinus(None=VertexInterpolator(VS=@n160.Z).PS), B=ScalarParameter() {defaultValue=10, parameterName=T3 - Bottom Top Length})
  n162 = ScalarParameter() {parameterName=T3 - Bottom Top Position}
  n163 = VertexInterpolator(VS=LocalPosition() {localOrigin=InstancePreSkinning}.Z)
  n164 = Add(
   A=Divide(A=OneMinus(None=@n163.PS), B=ScalarParameter() {defaultValue=35, parameterName=T3 - Bottom Top Length 2})
   B=ScalarParameter() {defaultValue=1.1, parameterName=T3 - Bottom Top Position 2}
  )
  n165 = OneMinus(None=@n164)
  n166 = Max(
   A=Multiply(A=Add(A=@n161, B=@n162), B=ScalarParameter() {defaultValue=1, parameterName=T3 - Bottom Top Intensity})
   B=Multiply(A=Min(A=@n164, B=@n165), B=ScalarParameter() {defaultValue=0.75, parameterName=T3 - Bottom Top Intensity 2})
  )
  n167 = SetMaterialAttributes!(
   MaterialAttributes=GetMaterialAttributes(None=@n114).MaterialAttributes
   EmissiveColor=LinearInterpolate(
    A=LinearInterpolate(A=@n115.RGB, B=@n122, Alpha=@n151)
    B=VectorParameter() {parameterName=T3 - Fresnel Color, defaultValue=<rgba 0.9301 0 0.4703 1>}.RGB
    Alpha=Max(A=@n159, B=@n166)
   )
   BaseColor=Constant()
   Roughness=ScalarParameter() {defaultValue=0.2, parameterName=T3 - Under Coat Roughness}
   ShadingModel=ShadingModel() {shadingModel=MSM_ClearCoat}
   CustomData0=Constant() {r=1}
   CustomData1=ScalarParameter() {defaultValue=0.4, parameterName=T3 - Clear Coat Roughness}
   Metallic=ScalarParameter() {defaultValue=0.4, parameterName=T3 - Metallic}
  ) {attributeSetTypes [
    #4
    #2
    #9
    #1
    #7
    #8
    #3
   ]}
  n168 = MaterialFunctionCall(Secondary=NamedRerouteUsage(), "MA (In)"=BlendMaterialAttributes(A=@n114, B=@n167, Alpha=NamedRerouteUsage())) {materialFunction=&MF_SL_37_SecondaryGlass}
  n169 = StaticBoolParameter() {group=SL 37 - Tertiary, parameterName=T3 - Use Clear Coat}
  n170 = StaticBoolParameter() {defaultValue=true, group=SL 37 - Tertiary, parameterName=T3 - Use Color Curve}
  n171 = ScalarParameter() {defaultValue=0.2, group=SL 37 - Tertiary, parameterName=T3 Tertiary - Color Curve Distortion}
  n172 = ScalarParameter() {parameterName=T3 Tertiary - Metallic}
  n173 = MaterialFunctionCall(
   Material Attributes=StaticSwitchParameter(
    True=MaterialFunctionCall(
     Secondary - Curve Amount=@n8
     Color Gradient Mask=@n9
     Tertiary=@n10
     Tertiary Fresnel Color=@n11.RGB
     Secondary - Curve Index=@n12
     "MA (In)"=@n168."Switch Param (False)"
     Use Clear Coat=@n169
     Use Color Curve=@n170
     Distortion Amount=@n171
     Metallic=@n172
    ) {materialFunction=&MF_SL_37_TertiaryFresnel}."MA (Out)"
    False=@n168."Switch Param (False)"
   ) {defaultValue=true, parameterName=Use Tertiary}
   Color=VectorParameter() {parameterName=T3 Decals - Emissive Color, defaultValue=<rgba 2 0.0112 0.6463 1>}.RGB
   Fresnel Exp=ScalarParameter() {defaultValue=0.1353, parameterName=T3 Decal - Fresnel Exp}
   Mask=NamedRerouteUsage()
   Intensity=ScalarParameter() {defaultValue=1, parameterName=T3 Decal - Intensity}
  ) {materialFunction=&MF_SL_37_Decals}
  FunctionOutput(None=StaticSwitchParameter(
    True=MaterialFunctionCall(Additional Mask=@n5, Outlines=@n6, Outline Color=@n7.RGB, "MA (In)"=@n173.BlendMaterialAttributes) {materialFunction=&MF_SL_37_Outlines}.SetMaterialAttributes
    False=@n104
   ) {group=SL 37 - Global, parameterName=Tier 3 - Beetle}) {outputName="MA (Out)"}
  n174 = CollectionParameter() {
   collection=&FortniteMaterialParameters
   parameterName=SunlightDirectionalVector
  }
  n175 = Normalize(VectorInput=Add(A=CameraVectorWS(), B=ComponentMask(None=@n174) {b=true, g=true, r=true}))
  n176 = Normalize(VectorInput=VertexNormalWS())
  n177 = Multiply(
   A=SmoothStep(Min=ScalarParameter() {defaultValue=0.99, parameterName=Eye Highlight Size}, Value=DotProduct(A=@n175, B=@n176) {desc=Blinn}) {constMin=0.995}
   B=ScalarParameter() {defaultValue=0.2, parameterName=Eye Highlight Strength}
  )
  LinearInterpolate(Alpha=@n177) {constA=0.5, constB=0.2}
  n178 = Ceil(None=AppendVector(A=@n3, B=ComponentMask(None=@n1) {g=true}))
  n179 = ComponentMask(None=VectorNoise(World Position=AppendVector(A=@n178, B=@n2))) {r=true}
  MaterialFunctionCall!(Tex=@n179, Position=ScalarParameter() {defaultValue=0.7, parameterName="Position (S)"}, Range=Constant() {r=1}) {materialFunction=/BRCosmetics/Characters/Player/Male/Medium/Bodies/M_MED_PawJasmineGravity/Skins/Greet/Materials/MF_HistogramSelect}
  n180 = ScalarParameter() {defaultValue=-0.7, parameterName=Sine Smooth Step}
  n181 = ScalarParameter() {defaultValue=1, parameterName=Sine Smooth}
  n182 = Multiply(A=ComponentMask(None=@n132) {r=true}, B=VectorParameter() {parameterName=Distort Tiling, defaultValue=<rgba 15 5 0 0>}.R)
  n183 = Multiply(A=Sine(None=@n182), B=Cosine(None=@n182))
  n184 = Multiply(A=Add(A=SmoothStep(Min=@n180, Max=@n181, Value=@n183), B=@n135), B=ScalarParameter() {defaultValue=0.6, parameterName=Distortion Intsense})
  n185 = Add(A=@n184, B=@n130)
  MaterialFunctionCall!(
   Width=ScalarParameter() {defaultValue=0.9, parameterName=TestWidth}
   Sharpness=ScalarParameter() {defaultValue=0.8, parameterName=TestSmooth}
   Offset
   Direction Switch
   Compare=Constant()
   Bands=ScalarParameter() {defaultValue=0.5, parameterName=TestBands}
   Input Coordinates=@n185
  ) {materialFunction=/Engine/Functions/Engine_MaterialFunctions03/Procedurals/GeneratedOffsetBands}
  NamedRerouteDeclaration(None=@n79) {name=Vest, nodeColor=<rgba 0 1 0.5547 1>}
  NamedRerouteDeclaration(None=@n79) {name=Fresnel Fade Mask, nodeColor=<rgba 0 1 0.1797 1>}
  n186 = MaterialFunctionCall(Float4=@n25) {materialFunction=&BreakOutFloat4Components}
  n187 = NamedRerouteDeclaration(None=Multiply(A=@n186.R, B=ScalarParameter() {defaultValue=1, group=SL 37 - Global, parameterName=Decal - Global Opacity})) {name=Decals, nodeColor=<rgba 1 1 0 1>}
  n188 = NamedRerouteDeclaration(None=@n186.G) {name=Outlines, nodeColor=<rgba 0 1 0 1>}
  n189 = Step(X=@n186.B) {constX=0.9, constY=0.05}
  NamedRerouteDeclaration(None=Max(A=Max(A=@n187, B=@n188), B=NamedRerouteDeclaration(None=@n189) {name=Hair, nodeColor=<rgba 0.2188 0 1 1>})) {name=FX2, nodeColor=<rgba 0.125 0 1 1>}
  NamedRerouteDeclaration(None=@n24) {name=Original MA, nodeColor=<rgba 1 0 0.8672 1>}
  NamedRerouteDeclaration(None=@n26.Additional Mask) {name=Additional Mask, nodeColor=<rgba 1 0 0.9141 1>}
  NamedRerouteDeclaration(None=@n186.A) {name=Variant, nodeColor=<rgba 0.8516 0 1 1>}
  n190 = Max(
   A=Max(A=@n95, B=NamedRerouteDeclaration(None=@n94.G) {name=Secondary, nodeColor=<rgba 0 1 0 1>})
   B=NamedRerouteDeclaration(None=@n94.B) {name=Tertiary, nodeColor=<rgba 0 0 1 1>}
  )
  NamedRerouteDeclaration(None=Max(A=@n190, B=NamedRerouteDeclaration(None=@n94.A) {name=Color Gradient Mask, nodeColor=<rgba 1 1 1 1>})) {name=Combined, nodeColor=<rgba 0 0.625 1 1>}
  n191 = Multiply(A=Time()) {constB=0.1}
  NamedRerouteDeclaration(None=Subtract(A=@n179, B=Sine(None=@n191))) {name=Sine Noise, nodeColor=<rgba 0 1 0.4375 1>}
  n192 = ComponentMask(None=VertexInterpolator(VS=LocalPosition() {localOrigin=InstancePreSkinning}.XYZ).PS) {b=true}
  n193 = ScalarParameter() {defaultValue=151.0474, group=SL37 - T2 Spider, parameterName=Preskinned Mask - Length}
  NamedRerouteDeclaration(None=Saturate(None=Divide(A=@n192, B=@n193))) {name=Preskinned Mask, nodeColor=<rgba 0.1953 0 1 1>}
  NamedRerouteDeclaration(None=@n190) {name=FX1, nodeColor=<rgba 1 0.4453 0 1>}
  NamedRerouteUsage()
  NamedRerouteUsage()
  NamedRerouteUsage()
  OneMinus(None=@n177)
  PixelNormalWS()
  ScalarParameter() {defaultValue=1, parameterName=Firefly - Random Speed}
  ScalarParameter() {defaultValue=1, parameterName=Firefly - Random Speed}
  SceneTexture() {sceneTextureId=PPI_BaseColor}
  Sine(None=@n36)
  TextureCoordinate()
  TextureSample(UVs=@n185) {texture=/BRCosmetics/Weapons/WeaponSkins/Textures/T_SimpleSIne2}
  Transform(None=@n184)
 ]}
