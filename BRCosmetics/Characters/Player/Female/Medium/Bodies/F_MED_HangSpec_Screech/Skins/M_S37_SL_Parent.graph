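# Parent skeletal-mesh character material for the Screech female S37 SL skins, layering custom hair shading over the shared 2023 character pipeline.
# PreFX output runs through MF_S37_SL_Parent, then the Hair 2023 function, whose desaturated color is retinted with a flippable Tech Art mask gradient.
# Two Fresnel silk bands and a sun-and-view-dependent rim feed emissive.
# A stepped Tech Art mask blends hair in, behind a static switch, before PostFX.
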Material /BRCosmetics/Characters/Player/Female/Medium/Bodies/F_MED_HangSpec_Screech/Skins/M_S37_SL_Parent
graph {
 properties {
  bAutomaticallySetUsageInEditor=false
  bHasPixelAnimation=true
  bUseMaterialAttributes=true
  bUsedWithMeshDeformer=true
  bUsedWithMorphTargets=true
  bUsedWithSkeletalMesh=true
  shadingModel=MSM_FromMaterialExpression
  translucencyLightingMode=TLM_SurfacePerPixelLighting
  usedShadingModels=Default Lit | Subsurface | Clear Coat | Subsurface Profile | Cloth
 }
 nodes [
  Constant()
  Constant()
  Constant() {r=1}
  n1 = ScalarParameter() {defaultValue=5, group=SL 37 - Hair, parameterName=Hair Outer Fresnel - Exp}
  Fresnel(ExponentIn=@n1) {baseReflectFraction=0}
  n2 = MaterialFunctionCall() {materialFunction=&MF_CharacterParent_2023_PreFX}
  n3 = ComponentMask(None=@n2.M) {r=true}
  n4 = MaterialFunctionCall!(M=@n3, Skin Mask=ComponentMask(None=@n2.M) {b=true}, Result=@n2.Result, Tech Art Mask=AppendVector(A=@n2.Tech Art Mask, B=@n2.Tech Art Mask)) {materialFunction=&/MF_S37_SL_Parent}
  n5 = GetMaterialAttributes(None=MaterialFunctionCall!(In=@n4."MA (Out)", Mask Input=@n2.M) {materialFunction=/Game/Global/Materials/Character_Parent_2023/Hair_Parent_2023/MaterialFunctions/MF_FN_Hair_2023}.Result) {attributeGetTypes [#1]}
  n6 = Desaturation(None=@n5.Base Color)
  n7 = ScalarParameter() {defaultValue=0.3, group=SL 37 - Hair, parameterName=T3 - Silk Step}
  n8 = Add(A=@n7, B=ScalarParameter() {defaultValue=0.584, group=SL 37 - Hair, parameterName=T3 - Silk Smoothness})
  n9 = OneMinus(None=Fresnel(ExponentIn=ScalarParameter() {defaultValue=0.37, group=SL 37 - Hair, parameterName=T3 - Silk}))
  n10 = ScalarParameter() {defaultValue=0.7, group=SL 37 - Hair, parameterName=T3 - Silk Step 2}
  n11 = Add(A=@n10, B=ScalarParameter() {defaultValue=0.244, group=SL 37 - Hair, parameterName=T3 - Silk Smoothness 2})
  n12 = OneMinus(None=Fresnel(ExponentIn=ScalarParameter() {defaultValue=0.2, group=SL 37 - Hair, parameterName=T3 - Silk 2}))
  n13 = Max(
   A=Multiply(A=SmoothStep(Min=@n7, Max=@n8, Value=@n9), B=ScalarParameter() {defaultValue=0.25, group=SL 37 - Hair, parameterName=T3 - Silk Intensity})
   B=Multiply(
    A=SmoothStep(Min=@n10, Max=@n11, Value=@n12)
    B=ScalarParameter() {defaultValue=0.25, group=SL 37 - Hair, parameterName=T3 - Silk Intensity 2}
   )
  )
  n14 = Multiply(A=@n6, B=Multiply(A=@n3, B=@n13))
  n15 = VectorParameter() {group=SL 37 - Hair, parameterName=T3 - Spec Light Color, defaultValue=<rgba 5.3218 10 1.875 1>}
  n16 = CameraVectorWS()
  n17 = AtmosphericLightVector()
  n18 = DotProduct(A=VertexNormalWS(), B=@n16)
  n19 = Multiply(A=DotProduct(A=@n16, B=@n17), B=Power(Base=OneMinus(None=@n18), Exp=@n1))
  n20 = VectorParameter() {group=SL 37 - Hair, parameterName=T3 - Hair Color Highlight, defaultValue=<rgba 0.4287 0.956 0.0395 1>}
  n21 = ComponentMask(None=@n2.Tech Art Mask) {b=true}
  n22 = OneMinus(None=@n21)
  n23 = ScalarParameter() {group=SL 37 - Hair, parameterName=Flip Hair Gradient}
  n24 = Power(
   Base=Multiply(
    A=LinearInterpolate(A=@n21, B=@n22, Alpha=@n23)
    B=ScalarParameter() {defaultValue=2.6, group=SL 37 - Hair, parameterName=Hair Highlight Offset}
   )
   Exp=ScalarParameter() {defaultValue=5, group=SL 37 - Hair, parameterName=Hair Highlight Length}
  )
  n25 = Multiply(
   A=@n6
   B=LinearInterpolate(
    A=VectorParameter() {group=SL 37 - Hair, parameterName=T3 - Hair Color, defaultValue=<rgba 0.1683 0.4342 0.0194 1>}.RGB
    B=@n20.RGB
    Alpha=Saturate(None=@n24)
   )
  )
  n26 = ScalarParameter() {defaultValue=1, group=SL 37 - Hair, parameterName=T3 - Hair Color Intensity}
  n27 = SetMaterialAttributes!(
   MaterialAttributes=@n5.MaterialAttributes
   Emissive Color=Max(A=Multiply(A=@n14, B=@n15.RGB), B=Multiply(A=@n19, B=@n20.RGB))
   Base Color=Multiply(A=@n3, B=Multiply(A=@n25, B=@n26))
   Fuzz Color or Subsurface Color
   Roughness=ScalarParameter() {defaultValue=0.4, group=SL 37 - Hair, parameterName=Hair Roughness}
   Specular
   Metallic
  ) {attributeSetTypes [
    #2
    #1
    #3
    #4
    #5
    #6
   ]}
  n28 = Step(X=@n21) {constX=0.9, constY=0.05}
  MaterialFunctionCall(
   Material=StaticSwitchParameter(True=BlendMaterialAttributes(A=@n4."MA (Out)", B=@n27, Alpha=@n28), False=@n4."MA (Out)") {defaultValue=true, parameterName=Use SL37 Hair}
   M=@n2.M
   Pre Skinned Local Position=@n2.M
   Distance Blend=@n2.M
   Fresnel=@n2.M
  ) {materialFunction=&MF_CharacterParent_2023_PostFX}
  Multiply(A=@n13, B=DotProduct(A=@n17, B=VertexNormalWS()))
  NamedRerouteUsage()
  ScalarParameter() {defaultValue=1, group=SL 37 - Hair, parameterName=T3 - Hair Roughness}
 ]
}
